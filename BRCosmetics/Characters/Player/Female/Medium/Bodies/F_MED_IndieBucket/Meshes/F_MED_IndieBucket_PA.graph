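# Physics asset for the F_MED_IndieBucket female player body, setting up secondary motion for its fanny pack and left-side overall flaps, buckle and shoulder strap.
# Pelvis, spine and thigh capsules form the body volumes, while small master spheres anchor each dangling chain to pelvis or spine_05 with free-rotation constraints.
# Chain links use angular limits, and a free-rotation constraint ties the strap's end to the back flap's tip, closing a loop.
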PhysicsAsset /BRCosmetics/Characters/Player/Female/Medium/Bodies/F_MED_IndieBucket/Meshes/F_MED_IndieBucket_PA
bodies {
 dyn_bk_l_overall_flap_01 [{shapeType=Capsule, length=2, radius=3, shapeName=dyn_bk_l_overall_flap_01_capsule, center=<3 0 4>, rotation=<rot 90 0 0>}]
 dyn_bk_l_overall_flap_02 [{shapeType=Capsule, length=2, radius=3, shapeName=dyn_bk_l_overall_flap_01_capsule, center=<3 -4 2>, rotation=<rot 90 0 0>}]
 dyn_bk_l_overall_flap_mstr [{shapeType=Sphere, radius=0.505, shapeName=dyn_bk_l_overall_flap_mstr_sphere}]
 dyn_fanny_pack_01 [{shapeType=Capsule, length=12, radius=6, shapeName=dyn_fanny_pack_01_capsule, rotation=<rot 0 0 90>}]
 dyn_fanny_pack_mstr [{shapeType=Sphere, radius=0.505, shapeName=dyn_fanny_pack_mstr_sphere}]
 dyn_fr_l_overall_buckle_01 [{shapeType=Capsule, length=2, radius=3.5, shapeName=dyn_fr_l_overall_buckle_01_capsule, center=<3 2 2>, rotation=<rot 90 0 0>}]
 dyn_fr_l_overall_buckle_02 [{shapeType=Capsule, length=2, radius=3.5, shapeName=dyn_fr_l_overall_buckle_01_capsule, center=<3 0.5 3>, rotation=<rot 90 0 0>}]
 dyn_l_overal_flap_01 [{shapeType=Capsule, length=4.5, radius=4, shapeName=dyn_l_overal_flap_01_capsule, center=<4 0 3>, rotation=<rot 90 0 0>}]
 dyn_l_overal_flap_01_mstr [{shapeType=Sphere, radius=0.505, shapeName=dyn_l_overal_flap_01_mstr_sphere}]
 dyn_l_shoulder_strap_01 [{shapeType=Capsule, length=5, radius=4, shapeName=dyn_bk_l_overall_flap_01_capsule, center=<3 0 4.5>, rotation=<rot 0 0 90>}]
 dyn_l_shoulder_strap_02 [{shapeType=Capsule, length=2, radius=4, shapeName=dyn_bk_l_overall_flap_01_capsule, center=<3 0 4.5>, rotation=<rot 0 0 90>}]
 dyn_l_shoulder_strap_03 [{shapeType=Capsule, length=2, radius=4, shapeName=dyn_bk_l_overall_flap_01_capsule, center=<3 0 4.5>, rotation=<rot 9 0 90>}]
 dyn_l_shoulder_strap_mstr [{shapeType=Sphere, radius=0.505, shapeName=dyn_l_shoulder_strap_mstr_sphere}]
 pelvis [{shapeType=Capsule, length=10, radius=10, center=<0 -4 0>}]
 spine_01 [{shapeType=Capsule, length=6, radius=10, shapeName=spine_01_capsule, center=<0 -2.5 0>}]
 spine_02 [{shapeType=Capsule, length=10, radius=8, shapeName=spine_02_capsule}]
 spine_03 [{shapeType=Capsule, length=8, radius=9, shapeName=spine_03_capsule, center=<0 1 0>}]
 spine_04 [{shapeType=Capsule, length=6, radius=10.5, shapeName=spine_04_capsule}]
 spine_05 [{shapeType=Capsule, length=10, radius=10, shapeName=spine_05_capsule}]
 thigh_l [{shapeType=Capsule, length=20, radius=8.5, center=<-17 -1 0>, rotation=<rot 90 -175 -170>}]
 thigh_r [{shapeType=Capsule, length=20, radius=9, center=<17 0 0>, rotation=<rot 90 0 0>}]
}
constraints [
 {bone1Name=thigh_l, bone2Name=pelvis, swing1LimitDegrees=45, swing2LimitDegrees=45, twistLimitDegrees=45}
 {bone1Name=thigh_r, bone2Name=pelvis, swing1LimitDegrees=45, swing2LimitDegrees=45, twistLimitDegrees=45}
 {
  bone1Name=dyn_fanny_pack_mstr
  bone2Name=pelvis
  swing1LimitDegrees=45
  swing1Motion=Free
  swing2LimitDegrees=45
  swing2Motion=Free
  twistLimitDegrees=45
  twistMotion=Free
 }
 {bone1Name=dyn_fanny_pack_01, bone2Name=dyn_fanny_pack_mstr, swing1LimitDegrees=1, swing2LimitDegrees=15, twistLimitDegrees=1}
 {
  bone1Name=spine_01
  bone2Name=pelvis
  swing1LimitDegrees=45
  swing1Motion=Free
  swing2LimitDegrees=45
  swing2Motion=Free
  twistLimitDegrees=45
  twistMotion=Free
 }
 {
  bone1Name=spine_02
  bone2Name=spine_01
  swing1LimitDegrees=45
  swing1Motion=Free
  swing2LimitDegrees=45
  swing2Motion=Free
  twistLimitDegrees=45
  twistMotion=Free
 }
 {
  bone1Name=spine_03
  bone2Name=spine_02
  swing1LimitDegrees=45
  swing1Motion=Free
  swing2LimitDegrees=45
  swing2Motion=Free
  twistLimitDegrees=45
  twistMotion=Free
 }
 {
  bone1Name=spine_04
  bone2Name=spine_03
  swing1LimitDegrees=45
  swing1Motion=Free
  swing2LimitDegrees=45
  swing2Motion=Free
  twistLimitDegrees=45
  twistMotion=Free
 }
 {
  bone1Name=spine_05
  bone2Name=spine_04
  swing1LimitDegrees=45
  swing1Motion=Free
  swing2LimitDegrees=45
  swing2Motion=Free
  twistLimitDegrees=45
  twistMotion=Free
 }
 {
  bone1Name=dyn_bk_l_overall_flap_mstr
  bone2Name=spine_05
  swing1LimitDegrees=45
  swing1Motion=Free
  swing2LimitDegrees=45
  swing2Motion=Free
  twistLimitDegrees=45
  twistMotion=Free
 }
 {bone1Name=dyn_bk_l_overall_flap_01, bone2Name=dyn_bk_l_overall_flap_mstr, swing1LimitDegrees=30, swing2LimitDegrees=60, twistLimitDegrees=5}
 {
  bone1Name=dyn_l_overal_flap_01_mstr
  bone2Name=spine_05
  swing1LimitDegrees=45
  swing1Motion=Free
  swing2LimitDegrees=45
  swing2Motion=Free
  twistLimitDegrees=45
  twistMotion=Free
 }
 {bone1Name=dyn_l_overal_flap_01, bone2Name=dyn_l_overal_flap_01_mstr, swing1LimitDegrees=5, swing2LimitDegrees=30, twistLimitDegrees=5}
 {bone1Name=dyn_fr_l_overall_buckle_01, bone2Name=dyn_l_overal_flap_01, swing1LimitDegrees=30, swing2LimitDegrees=30, twistLimitDegrees=1}
 {
  bone1Name=dyn_fr_l_overall_buckle_02
  bone2Name=dyn_fr_l_overall_buckle_01
  swing1LimitDegrees=30
  swing2LimitDegrees=30
  twistLimitDegrees=1
 }
 {
  bone1Name=dyn_l_shoulder_strap_mstr
  bone2Name=spine_05
  swing1LimitDegrees=45
  swing1Motion=Free
  swing2LimitDegrees=45
  swing2Motion=Free
  twistLimitDegrees=45
  twistMotion=Free
 }
 {bone1Name=dyn_l_shoulder_strap_01, bone2Name=dyn_l_shoulder_strap_mstr, swing1LimitDegrees=15, swing2LimitDegrees=30, twistLimitDegrees=1}
 {bone1Name=dyn_l_shoulder_strap_02, bone2Name=dyn_l_shoulder_strap_01, swing1LimitDegrees=15, swing2LimitDegrees=30, twistLimitDegrees=1}
 {bone1Name=dyn_l_shoulder_strap_03, bone2Name=dyn_l_shoulder_strap_02, swing1LimitDegrees=15, swing2LimitDegrees=30, twistLimitDegrees=1}
 {bone1Name=dyn_bk_l_overall_flap_02, bone2Name=dyn_bk_l_overall_flap_01, swing1LimitDegrees=15, swing2LimitDegrees=15, twistLimitDegrees=5}
 {
  bone1Name=dyn_bk_l_overall_flap_02
  bone2Name=dyn_l_shoulder_strap_03
  swing1LimitDegrees=45
  swing1Motion=Free
  swing2LimitDegrees=45
  swing2Motion=Free
  twistLimitDegrees=45
  twistMotion=Free
 }
]
