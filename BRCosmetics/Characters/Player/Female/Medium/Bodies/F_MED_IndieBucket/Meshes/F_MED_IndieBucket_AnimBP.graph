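# Animation Blueprint for the female IndieBucket cosmetic body, adding secondary motion to clothing and accessories atop the parent mesh's pose.
# Copied pose runs through RigidBody, Control Rig, crouch constraints and AnimDynamics for zipper, sleeves, pants and shoes into a cache.
# Output selects cached, melee-dependent RigidBody or weighted-constraint fallback poses by rigid-body state.
# EventGraph updates gravity, front-end alpha, cloth WPO/wind material parameters and pistol/rifle flags.
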AnimBlueprint /BRCosmetics/Characters/Player/Female/Medium/Bodies/F_MED_IndieBucket/Meshes/F_MED_IndieBucket_AnimBP
target_skeleton=&/F_MED_IndieBucket_Skeleton
graphs [
 AnimGraph {nodes [
   Anim Dynamics - Bone: dyn_fanny_pack_01(Alpha=1, ExternalForce="0,0,0") {
    angularDampingOverride=0.9999
    angularSpringConstant=75
    bAngularSpring=true
    bOverrideAngularDamping=true
    bUseSphericalLimits=true
    simulationSpace=RootRelative
    alphaBoolBlend {blendOption=Linear}
    boundBone {boneName=dyn_fanny_pack_01}
    componentLinearAccScale=<0.6 0.6 0.6>
    componentLinearVelScale=<0.6 0.6 0.6>
    physicsBodyDefinitions [{
      boundBone {boneName=dyn_fanny_pack_01}
      boxExtents=<10 20 20>
      constraintSetup {angularLimitsMax=<1 15 1>, angularLimitsMin=<-1 -15 -1>, angularTarget=<0 0 0>}
      localJointOffset=<1 0 0>
     }]
   }
   n1 = RigidBody!(ComponentPose=Local To Component(LocalPose=Use cached pose 'PreCache'()), Alpha=0.1, OverrideWorldGravity:Vector="0,0,-980", ExternalForce) -> (Pose) {
    bClampLinearTranslationLimitToRefPose=true
    bOverrideWorldGravity=true
    overridePhysicsAsset=&/F_MED_IndieBucket_PA
    simulationSpace=BaseBoneSpace
    alphaBoolBlend {blendOption=Linear}
    baseBoneRef {boneName=pelvis}
    componentAppliedLinearAccClamp=<3000 3000 3000>
    componentLinearAccScale=<0.5 0.1 0.3>
    componentLinearVelScale=<0.25 0.1 0.3>
    simSpaceSettings {
     maxAngularAcceleration=1000
     maxAngularVelocity=3
     maxLinearAcceleration=700
     maxLinearVelocity=500
     velocityScaleZ=5
     worldAlpha=1
     externalLinearDragV=<0.2 0.2>
    }
   }
   n2 = Component To Local(ComponentPose=@n1)
   n3 = Use cached pose 'PreCache'()
   n4 = Get bIsPlayingMeleeAnim()
   n5 = "Constraint - dyn_fr_l_overall_buckle_01_mstr (PPPPPPPP)"!(
    ComponentPose=Local To Component(LocalPose=Use cached pose 'PreCache'())
    Alpha=1
    ConstraintWeights_0=0.0926
    ConstraintWeights_1:float=0.0408
    ConstraintWeights_2:float=0.3532
    ConstraintWeights_3:float=0.1161
    ConstraintWeights_4:float=0.396
    ConstraintWeights_5:float=0
    ConstraintWeights_6:float=0.0013
    ConstraintWeights_7:float=0
   ) -> (Pose) {
    alphaBoolBlend {blendOption=Linear}
    boneToModify {boneName=dyn_fr_l_overall_buckle_01_mstr}
    constraintSetup [
     {targetBone {boneName=pelvis}}
     {targetBone {boneName=spine_01}}
     {targetBone {boneName=spine_02}}
     {targetBone {boneName=spine_03}}
     {targetBone {boneName=spine_04}}
     {targetBone {boneName=spine_05}}
     {targetBone {boneName=clavicle_l}}
     {targetBone {boneName=thigh_l}}
    ]
    constraintWeights [1,1,1,1,1,1,1,1]
   }
   n6 = "Constraint - dyn_l_overal_flap_01_mstr (PPPPPP)"!(
    ComponentPose=@n5
    Alpha=1
    ConstraintWeights_0=0.0115
    ConstraintWeights_1:float=0.0814
    ConstraintWeights_2:float=0.071
    ConstraintWeights_3:float=0.7849
    ConstraintWeights_4:float=0
    ConstraintWeights_5:float=0.0512
   ) -> (Pose) {
    alphaBoolBlend {blendOption=Linear}
    boneToModify {boneName=dyn_l_overal_flap_01_mstr}
    constraintSetup [
     {targetBone {boneName=spine_01}}
     {targetBone {boneName=spine_02}}
     {targetBone {boneName=spine_03}}
     {targetBone {boneName=spine_04}}
     {targetBone {boneName=spine_05}}
     {targetBone {boneName=clavicle_l}}
    ]
    constraintWeights [1,1,1,1,1,1]
   }
   n7 = "Constraint - dyn_l_shoulder_strap_03 (PPPPPPPP)"!(
    ComponentPose=@n6
    Alpha=1
    ConstraintWeights_0=0.8217
    ConstraintWeights_1:float=0.014
    ConstraintWeights_2:float=0.1265
    ConstraintWeights_3:float=0.0037
    ConstraintWeights_4:float=0.0082
    ConstraintWeights_5:float=0.0053
    ConstraintWeights_6:float=0.0204
    ConstraintWeights_7:float=0.0001
   ) -> (Pose) {
    alphaBoolBlend {blendOption=Linear}
    boneToModify {boneName=dyn_l_shoulder_strap_03}
    constraintSetup [
     {targetBone {boneName=pelvis}}
     {targetBone {boneName=spine_01}}
     {targetBone {boneName=spine_02}}
     {targetBone {boneName=spine_03}}
     {targetBone {boneName=spine_04}}
     {targetBone {boneName=thigh_l}}
     {targetBone {boneName=thigh_twist_01_l}}
     {targetBone {boneName=thigh_r}}
    ]
    constraintWeights [1,1,1,1,1,1,1,1]
   }
   n8 = "Transform (Modify) Bone - Bone: dyn_bk_l_overall_flap_mstr"(ComponentPose=@n7, Alpha=1, Translation="0,0,0", Rotation="-10.000,0.000,0.000", Scale="1,1,1") {
    rotationMode=BMM_Additive
    rotationSpace=BCS_BoneSpace
    translationMode=BMM_Additive
    translationSpace=BCS_BoneSpace
    alphaBoolBlend {blendOption=Linear}
    boneToModify {boneName=dyn_bk_l_overall_flap_mstr}
    rotation=<rot -10.0001 0 0>
   }
   Output Pose(Result=Blend Poses by bool(
     BlendPose_0=Blend Poses by bool(BlendPose_0=@n2, BlendPose_1=@n3, bActiveValue=@n4) {blendType=Linear}
     BlendPose_1=Component To Local(ComponentPose=@n8)
     bActiveValue=Get bRigidBodyEnabled()
    ) {blendType=Linear}) {layerGroup=None}
   n9 = Get FrontEndAlpha() -> (FrontEndAlpha:double)
   n10 = "Constraint - dyn_r_wrist_ring_top_mstr (R)"(ComponentPose=Local To Component(LocalPose=Copy Pose From Mesh() {bUseAttachedParent=true}), Alpha=@n9, ConstraintWeights_0=1) {
    alphaBoolBlend {blendOption=Linear}
    boneToModify {boneName=dyn_r_wrist_ring_top_mstr}
    constraintSetup [{transformType=Rotation, targetBone {boneName=hand_r}}]
    constraintWeights [1]
   }
   n11 = "Constraint - dyn_r_wrist_ring_bottom_mstr (R)"(ComponentPose=@n10, Alpha=@n9, ConstraintWeights_0=1) {
    alphaBoolBlend {blendOption=Linear}
    boneToModify {boneName=dyn_r_wrist_ring_bottom_mstr}
    constraintSetup [{transformType=Rotation, targetBone {boneName=hand_r}}]
    constraintWeights [1]
   }
   n12 = RigidBody(ComponentPose=@n11, OverrideWorldGravity="0,0,-980") {
    bClampLinearTranslationLimitToRefPose=true
    bOverrideWorldGravity=true
    overridePhysicsAsset=&F_MED_IndieBucket_PA
    simulationSpace=BaseBoneSpace
    alphaBoolBlend {blendOption=Linear}
    baseBoneRef {boneName=pelvis}
    componentAppliedLinearAccClamp=<3000 3000 3000>
    componentLinearAccScale=<0.5 0.1 0.3>
    componentLinearVelScale=<0.25 0.1 0.3>
    simSpaceSettings {
     maxAngularAcceleration=1000
     maxAngularVelocity=3
     maxLinearAcceleration=700
     maxLinearVelocity=500
     velocityScaleZ=5
     worldAlpha=1
     externalLinearDragV=<0.2 0.2>
    }
   }
   n13 = Control Rig!(Source=Component To Local(ComponentPose=@n12), Alpha, Is Crouching:bool=Get bIsCrouching(), Meele Animation:bool=Get bIsPlayingMeleeAnim()) -> (Pose) {
    alphaBoolBlend {blendOption=Linear}
    controlRigAssetReference {blueprintRigClass=&/F_MED_IndieBucket_ControlRig.F_MED_IndieBucket_ControlRig_C}
   }
   n14 = Get bIsCrouching()
   n15 = "Constraint - dyn_l_shoulder_strap_mstr (P)"!(ComponentPose=Local To Component(LocalPose=@n13), bAlphaBoolEnabled:bool=@n14, ConstraintWeights_0=1) -> (Pose) {
    alphaInputType=Bool
    alphaBoolBlend {blendOption=Linear}
    boneToModify {boneName=dyn_l_shoulder_strap_mstr}
    constraintSetup [{targetBone {boneName=pelvis}}]
    constraintWeights [1]
   }
   n16 = "Constraint - dyn_fr_l_overall_buckle_01_mstr (R)"!(ComponentPose=@n15, bAlphaBoolEnabled:bool=@n14, ConstraintWeights_0=1) -> (Pose) {
    alphaInputType=Bool
    alphaBoolBlend {blendOption=Linear}
    boneToModify {boneName=dyn_fr_l_overall_buckle_01_mstr}
    constraintSetup [{transformType=Rotation, targetBone {boneName=dyn_l_shoulder_strap_01}}]
    constraintWeights [1]
   }
   n17 = Anim Dynamics - Bone: dyn_fanny_pack_zipper_01(ComponentPose=@n16, Alpha=1, ExternalForce="0,0,0") {
    angularDampingOverride=0.9999
    bOverrideAngularDamping=true
    bUseGravityOverride=true
    bUsePlanarLimit=false
    simulationSpace=RootRelative
    alphaBoolBlend {blendOption=Linear}
    boundBone {boneName=dyn_fanny_pack_zipper_01}
    componentAppliedLinearAccClamp=<3000 3000 3000>
    componentLinearAccScale=<0.5 0.5 0.5>
    componentLinearVelScale=<0.25 0.25 0.25>
    gravityOverride {z=-980}
    physicsBodyDefinitions [{
      boundBone {boneName=dyn_fanny_pack_zipper_01}
      boxExtents=<2 10 10>
      constraintSetup {angularLimitsMax=<2 5 15>, angularLimitsMin=<-2 -30 -30>, angularTarget=<0 0 0>}
      localJointOffset=<3 0 0>
     }]
   }
   n18 = "Transform (Modify) Bone - Bone: r_sleeve_mstr"(ComponentPose=@n17, Alpha=0.5, Translation="-6.782,-5.343,11.634", Rotation="-10.000,0.000,-0.000", Scale="3,3,3") {
    rotationMode=BMM_Additive
    scaleMode=BMM_Additive
    translationMode=BMM_Additive
    alphaBoolBlend {blendOption=Linear}
    boneToModify {boneName=r_sleeve_mstr}
    rotation {pitch=-10}
    scale=<3 3 3>
    translation=<-6.782 -5.343 11.634>
   }
   n19 = "Transform (Modify) Bone - Bone: l_sleeve_mstr"(ComponentPose=@n18, Alpha=0.5, Translation="6.782,4.641,15.800", Rotation="-10.000,0.000,-0.000", Scale="3,3,3") {
    rotationMode=BMM_Additive
    scaleMode=BMM_Additive
    translationMode=BMM_Additive
    alphaBoolBlend {blendOption=Linear}
    boneToModify {boneName=l_sleeve_mstr}
    rotation {pitch=-10}
    scale=<3 3 3>
    translation=<6.782 4.641 15.8>
   }
   n20 = Anim Dynamics - Bone: l_sleeve_dyn(ComponentPose=@n19, Alpha=0.5, ExternalForce="0,0,0") {
    angularDampingOverride=0.9999
    bGravityOverrideInSimSpace=true
    bOverrideAngularDamping=true
    bUseGravityOverride=true
    simulationSpace=RootRelative
    alphaBoolBlend {blendOption=Linear}
    boundBone {boneName=l_sleeve_dyn}
    componentAppliedLinearAccClamp=<3000 3000 3000>
    componentLinearAccScale=<0.15 0.15 0.15>
    componentLinearVelScale=<0.15 0.15 0.15>
    gravityOverride {z=-980}
    physicsBodyDefinitions [{
      sphereCollisionRadius=0
      boundBone {boneName=l_sleeve_dyn}
      boxExtents=<12 6 12>
      constraintSetup {angularLimitsMax=<0 5 5>, angularLimitsMin=<0 -5 -5>, angularTarget=<0 0 0>}
      localJointOffset=<1 0 0>
     }]
    relativeSpaceBone {boneName=lowerarm_r}
   }
   n21 = Anim Dynamics - Bone: r_sleeve_dyn(ComponentPose=@n20, Alpha=0.5, ExternalForce="0,0,0") {
    angularDampingOverride=0.9999
    bGravityOverrideInSimSpace=true
    bOverrideAngularDamping=true
    bUseGravityOverride=true
    simulationSpace=RootRelative
    alphaBoolBlend {blendOption=Linear}
    boundBone {boneName=r_sleeve_dyn}
    componentAppliedLinearAccClamp=<3000 3000 3000>
    componentLinearAccScale=<0.15 0.15 0.15>
    componentLinearVelScale=<0.15 0.15 0.15>
    gravityOverride {z=-980}
    physicsBodyDefinitions [{
      sphereCollisionRadius=0
      boundBone {boneName=r_sleeve_dyn}
      boxExtents=<12 6 12>
      constraintSetup {angularLimitsMax=<0 5 5>, angularLimitsMin=<0 -5 -5>, angularTarget=<0 0 0>}
      localJointOffset=<1 0 0>
     }]
    relativeSpaceBone {boneName=lowerarm_r}
   }
   n22 = Anim Dynamics - Bone: l_pant_bottom_dyn(ComponentPose=@n21, Alpha=0.5, ExternalForce="0,0,0") {
    angularDampingOverride=0.998
    bGravityOverrideInSimSpace=true
    bOverrideAngularDamping=true
    bUseGravityOverride=true
    simulationSpace=RootRelative
    alphaBoolBlend {blendOption=Linear}
    boundBone {boneName=l_pant_bottom_dyn}
    componentAppliedLinearAccClamp=<3000 3000 3000>
    componentLinearAccScale=<0.3 0.3 0.3>
    componentLinearVelScale=<0.3 0.3 0.3>
    gravityOverride {z=-980}
    physicsBodyDefinitions [{
      sphereCollisionRadius=0
      boundBone {boneName=l_pant_bottom_dyn}
      boxExtents=<18 9 18>
      constraintSetup {angularLimitsMax=<0 5 5>, angularLimitsMin=<0 -5 -5>, angularTarget=<0 0 0>, linearAxesMax=<0 1 1>, linearAxesMin=<0 -1 -1>}
      localJointOffset=<1 0 0>
     }]
    relativeSpaceBone {boneName=lowerarm_r}
   }
   n23 = Anim Dynamics - Bone: r_pant_bottom_dyn(ComponentPose=@n22, Alpha=0.5, ExternalForce="0,0,0") {
    angularDampingOverride=0.998
    bGravityOverrideInSimSpace=true
    bOverrideAngularDamping=true
    bUseGravityOverride=true
    simulationSpace=RootRelative
    alphaBoolBlend {blendOption=Linear}
    boundBone {boneName=r_pant_bottom_dyn}
    componentAppliedLinearAccClamp=<3000 3000 3000>
    componentLinearAccScale=<0.3 0.3 0.3>
    componentLinearVelScale=<0.3 0.3 0.3>
    gravityOverride {z=-980}
    physicsBodyDefinitions [{
      sphereCollisionRadius=0
      boundBone {boneName=r_pant_bottom_dyn}
      boxExtents=<12 6 12>
      constraintSetup {angularLimitsMax=<0 5 5>, angularLimitsMin=<0 -5 -5>, angularTarget=<0 0 0>, linearAxesMax=<0 1 1>, linearAxesMin=<0 -1 -1>}
      localJointOffset=<1 0 0>
     }]
    relativeSpaceBone {boneName=lowerarm_r}
   }
   n24 = Anim Dynamics - Bone: dyn_l_shoelace_r_01(ComponentPose=@n23, Alpha=1, ExternalForce="0,0,0") {
    angularDampingOverride=1
    angularSpringConstant=1500
    bAngularSpring=true
    bGravityOverrideInSimSpace=true
    bOverrideAngularDamping=true
    bUsePlanarLimit=false
    gravityScale=0.35
    simulationSpace=RootRelative
    alphaBoolBlend {blendOption=Linear}
    boundBone {boneName=dyn_l_shoelace_r_01}
    physicsBodyDefinitions [{
      sphereCollisionRadius=0
      boundBone {boneName=dyn_l_shoelace_r_01}
      boxExtents=<10 20 20>
      constraintSetup {angularLimitsMax=<5 2 2>, angularLimitsMin=<-5 -25 -25>, angularTarget=<0 0 0>}
      localJointOffset=<10 0 0>
     }]
    relativeSpaceBone {boneName=foot_l}
   }
   n25 = Anim Dynamics - Bone: dyn_l_shoelace_l_01(ComponentPose=@n24, Alpha=1, ExternalForce="0,0,0") {
    angularDampingOverride=1
    angularSpringConstant=1500
    bAngularSpring=true
    bGravityOverrideInSimSpace=true
    bOverrideAngularDamping=true
    bUsePlanarLimit=false
    gravityScale=0.35
    simulationSpace=RootRelative
    alphaBoolBlend {blendOption=Linear}
    boundBone {boneName=dyn_l_shoelace_l_01}
    physicsBodyDefinitions [{
      sphereCollisionRadius=0
      boundBone {boneName=dyn_l_shoelace_l_01}
      boxExtents=<10 20 20>
      constraintSetup {angularLimitsMax=<5 2 2>, angularLimitsMin=<-5 -25 -25>, angularTarget=<0 0 0>}
      localJointOffset=<10 0 0>
     }]
    relativeSpaceBone {boneName=foot_l}
   }
   n26 = Anim Dynamics - Bone: dyn_l_shoe_l_ear_01(ComponentPose=@n25, Alpha=1, ExternalForce="0,0,0") {
    angularDampingOverride=0.9999
    angularSpringConstant=1600
    bAngularSpring=true
    bGravityOverrideInSimSpace=true
    bOverrideAngularDamping=true
    bUsePlanarLimit=false
    gravityScale=0.35
    simulationSpace=RootRelative
    alphaBoolBlend {blendOption=Linear}
    boundBone {boneName=dyn_l_shoe_l_ear_01}
    componentAppliedLinearAccClamp=<3000 3000 3000>
    componentLinearAccScale=<0.3 0.3 0.45>
    componentLinearVelScale=<0.15 0.15 0.225>
    physicsBodyDefinitions [{
      sphereCollisionRadius=0
      boundBone {boneName=dyn_l_shoe_l_ear_01}
      boxExtents=<10 20 20>
      constraintSetup {angularLimitsMax=<5 5 10>, angularLimitsMin=<-5 -15 -10>, angularTarget=<0 0 0>}
      localJointOffset=<-2 0 0>
     }]
    relativeSpaceBone {boneName=foot_l}
   }
   n27 = Anim Dynamics - Bone: dyn_l_shoe_r_ear_01(ComponentPose=@n26, Alpha=1, ExternalForce="0,0,0") {
    angularDampingOverride=0.9999
    angularSpringConstant=1500
    bAngularSpring=true
    bGravityOverrideInSimSpace=true
    bOverrideAngularDamping=true
    bUsePlanarLimit=false
    gravityScale=0.35
    simulationSpace=RootRelative
    alphaBoolBlend {blendOption=Linear}
    boundBone {boneName=dyn_l_shoe_r_ear_01}
    componentAppliedLinearAccClamp=<3000 3000 3000>
    componentLinearAccScale=<0.3 0.3 0.45>
    componentLinearVelScale=<0.15 0.15 0.225>
    physicsBodyDefinitions [{
      sphereCollisionRadius=0
      boundBone {boneName=dyn_l_shoe_r_ear_01}
      boxExtents=<10 20 20>
      constraintSetup {angularLimitsMax=<5 5 10>, angularLimitsMin=<-5 -15 -10>, angularTarget=<0 0 0>}
      localJointOffset=<-2 0 0>
     }]
    relativeSpaceBone {boneName=foot_l}
   }
   n28 = Anim Dynamics - Bone: dyn_r_shoelace_r_01(ComponentPose=@n27, Alpha=1, ExternalForce="0,0,0") {
    angularDampingOverride=1
    angularSpringConstant=1500
    bAngularSpring=true
    bGravityOverrideInSimSpace=true
    bOverrideAngularDamping=true
    bUsePlanarLimit=false
    gravityScale=0.35
    simulationSpace=RootRelative
    alphaBoolBlend {blendOption=Linear}
    boundBone {boneName=dyn_r_shoelace_r_01}
    physicsBodyDefinitions [{
      sphereCollisionRadius=0
      boundBone {boneName=dyn_r_shoelace_r_01}
      boxExtents=<10 20 20>
      constraintSetup {angularLimitsMax=<5 -6 -6>, angularLimitsMin=<-5 -25 -25>, angularTarget=<0 0 0>}
      localJointOffset=<10 0 0>
     }]
    relativeSpaceBone {boneName=foot_l}
   }
   n29 = Anim Dynamics - Bone: dyn_r_shoelace_l_01(ComponentPose=@n28, Alpha=1, ExternalForce="0,0,0") {
    angularDampingOverride=1
    angularSpringConstant=1500
    bAngularSpring=true
    bGravityOverrideInSimSpace=true
    bOverrideAngularDamping=true
    bUsePlanarLimit=false
    gravityScale=0.35
    simulationSpace=RootRelative
    alphaBoolBlend {blendOption=Linear}
    boundBone {boneName=dyn_r_shoelace_l_01}
    physicsBodyDefinitions [{
      sphereCollisionRadius=0
      boundBone {boneName=dyn_r_shoelace_l_01}
      boxExtents=<10 20 20>
      constraintSetup {angularLimitsMax=<5 -2 -2>, angularLimitsMin=<-5 -25 -25>, angularTarget=<0 0 0>}
      localJointOffset=<10 0 0>
     }]
    relativeSpaceBone {boneName=foot_l}
   }
   n30 = Anim Dynamics - Bone: dyn_r_shoe_l_ear_01(ComponentPose=@n29, Alpha=1, ExternalForce="0,0,0") {
    angularDampingOverride=0.9999
    angularSpringConstant=1600
    bAngularSpring=true
    bGravityOverrideInSimSpace=true
    bOverrideAngularDamping=true
    bUsePlanarLimit=false
    gravityScale=0.35
    simulationSpace=RootRelative
    alphaBoolBlend {blendOption=Linear}
    boundBone {boneName=dyn_r_shoe_l_ear_01}
    componentAppliedLinearAccClamp=<3000 3000 3000>
    componentLinearAccScale=<0.3 0.3 0.45>
    componentLinearVelScale=<0.15 0.15 0.225>
    physicsBodyDefinitions [{
      sphereCollisionRadius=0
      boundBone {boneName=dyn_r_shoe_l_ear_01}
      boxExtents=<10 20 20>
      constraintSetup {angularLimitsMax=<5 -5 10>, angularLimitsMin=<-5 -15 -10>, angularTarget=<0 0 0>}
      localJointOffset=<-2 0 0>
     }]
    relativeSpaceBone {boneName=foot_l}
   }
   n31 = Anim Dynamics - Bone: dyn_r_shoe_r_ear_01(ComponentPose=@n30, Alpha=1, ExternalForce="0,0,0") {
    angularDampingOverride=0.9999
    angularSpringConstant=1500
    bAngularSpring=true
    bGravityOverrideInSimSpace=true
    bOverrideAngularDamping=true
    bUsePlanarLimit=false
    gravityScale=0.35
    simulationSpace=RootRelative
    alphaBoolBlend {blendOption=Linear}
    boundBone {boneName=dyn_r_shoe_r_ear_01}
    componentAppliedLinearAccClamp=<3000 3000 3000>
    componentLinearAccScale=<0.3 0.3 0.45>
    componentLinearVelScale=<0.15 0.15 0.225>
    physicsBodyDefinitions [{
      sphereCollisionRadius=0
      boundBone {boneName=dyn_r_shoe_r_ear_01}
      boxExtents=<10 20 20>
      constraintSetup {angularLimitsMax=<5 5 10>, angularLimitsMin=<-5 -15 -10>, angularTarget=<0 0 0>}
      localJointOffset=<-2 0 0>
     }]
    relativeSpaceBone {boneName=foot_l}
   }
   Save cached pose 'PreCache'(Pose=Component To Local(ComponentPose=@n31)) {cacheName=PreCache}
  ]}
 EventGraph {nodes [
   Event BlueprintUpdateAnimation()
   > n1 = Sequence!(execute) -> (then_0, then_1, then_2:Exec, then_3:Exec, then_4:Exec)
   > n2 = Gravity Override State Function DEPRECATED(CharacterPartAnimInstance=Self-Reference(), GameTimeInSeconds=GetGameTimeInSeconds(), GravityOverrideParams=Get Gravity Override Params())
   > Set Gravity Override(Gravity Override=@n2)
   > Set FrontEndAlpha(FrontEndAlpha:double=SelectFloat(A=0.5, B=1, bPickA=Get bIsInFrontEnd())) -> (Output_Get:double)
   Is Valid(exec=@n1.then_1, InputObject=Get Player Pawn() -> (Player Pawn:Fort Player Pawn Athena Object Reference))
   > Set WPO Globals Function(
    Material=Get Mat Instance Dynamic()
    Overlay Material=Get Overlay Mat Instance Dynamic()
    Skeletal Mesh=Get Skel Mesh()
    "Local Noise Transform  Location"=Get Local Noise Transform Location() -> (Local Noise Transform Location:Vector)
    Local Noise Offset=Get Local Noise Offset()
   )
   > Set WPO Regional Function!(
    execute
    Material:Material Instance Dynamic Object Reference=Get Mat Instance Dynamic()
    Overlay Material:Material Instance Dynamic Object Reference=Get Overlay Mat Instance Dynamic()
    Skeletal Mesh:Skeletal Mesh Component Object Reference=Get Skel Mesh()
    Velocity Transform:Name=foot_r
    Velocity Interp Speed:double=Get Velocity Interp Speed()
    Collision Transform:Name=calf_r
    Collision Transform Axis:EAxis Enum
    Invert Collision Transform Axis:bool=false
    Delta Time X:double=0
    Previous Velocity Transform Location:"Vector (by ref)"=Get R1 Velocity Transform Location()
    Previous Velocity Transform Velocity:"Vector (by ref)"=Get R1 Velocity Transform Velocity()
    Region:int=1
   ) -> (then)
   > Set WPO Regional Function(
    Material=Get Mat Instance Dynamic()
    Overlay Material=Get Overlay Mat Instance Dynamic()
    Skeletal Mesh=Get Skel Mesh()
    Velocity Transform=hand_l
    Velocity Interp Speed=Get Velocity Interp Speed()
    Collision Transform=lowerarm_l
    Invert Collision Transform Axis=true
    Delta Time X=0
    Previous Velocity Transform Location=Get R2 Velocity Transform Location()
    Previous Velocity Transform Velocity=Get R2 Velocity Transform Velocity()
    Region=2
   )
   > n3 = Branch(Condition=Get bIsSkydiveFloating())
   > n4 = Set WPO Pho Wind Intensity(WPO Pho Wind Intensity=1)
   SetScalarParameterValue(execute=[@n4.then, @n5.then], self=Get Mat Instance Dynamic(), ParameterName=ClothSim_WindStrength, Value=Get WPO Pho Wind Intensity())
   n5 = Set WPO Pho Wind Intensity(execute=@n3.else, WPO Pho Wind Intensity=0)
   n6 = Cast To FortPlayerAnimInstance(execute=@n1.then_2, Object=GetAnimInstance(self=GetOwnerSkeletalMeshComponent()))
   > Set Pistol(Pistol:double=0) -> (Output_Get:double)
   > Set Rifle(Rifle:double=0) -> (Output_Get:double)
   > n7 = Switch on EFortWeaponCoreAnimation!(execute, Selection:EFortWeaponCoreAnimation Enum=Get WeaponCoreAnim(self=@n6)) -> (
    Melee:Exec
    Pistol:Exec
    Shotgun:Exec
    PaperBlueprint:Exec
    Rifle:Exec
    MeleeOneHand:Exec
    MachinePistol:Exec
    RocketLauncher:Exec
    GrenadeLauncher:Exec
    GoingCommando:Exec
    AssaultRifle:Exec
    TacticalShotgun:Exec
    SniperRifle:Exec
    TrapPlacement:Exec
    ShoulderLauncher:Exec
    AbilityDecoTool:Exec
    Crossbow:Exec
    C4:Exec
    RemoteControl:Exec
    DualWield:Exec
    AR_BullPup:Exec
    AR_ForwardGrip:Exec
    MedPackPaddles:Exec
    SMG_P90:Exec
    AR_DrumGun:Exec
    Consumable_Small:Exec
    Consumable_Large:Exec
    Balloon:Exec
    MountedTurret:Exec
    CreativeTool:Exec
    ExplosiveBow:Exec
    AshtonIndigo:Exec
    AshtonChicago:Exec
    MeleeDualWield:Exec
    Unarmed:Exec
   )
   Branch(execute=[@n7.AssaultRifle, @n7.Crossbow, @n7.GrenadeLauncher, @n7.Rifle, @n7.RocketLauncher, @n7.Shotgun, @n7.SniperRifle, @n7.TacticalShotgun])
   > Set Rifle(Rifle=1)
   > n8 = Branch()
   > Set Is Aiming Rifle(Is Aiming Rifle:bool=true) -> (Output_Get:bool)
   Set Is Aiming Rifle(execute=@n8.else, Is Aiming Rifle=false)
   Branch(execute=[@n7.MachinePistol, @n7.Pistol])
   > Set Pistol(Pistol=1)
   > n9 = Branch()
   > Set Is Aiming Pistol(Is Aiming Pistol:bool=true) -> (Output_Get:bool)
   Set Is Aiming Pistol(execute=@n9.else, Is Aiming Pistol=false)
   Event BlueprintInitializeAnimation()
   > Set Skel Mesh(Skel Mesh=GetOwnerSkeletalMeshComponent())
   n10 = TryGetPawnOwner()
   > Is Valid(InputObject=@n10)
   > n11 = Cast To FortPlayerPawnAthena(Object=@n10)
   > Set Player Pawn(Player Pawn:Fort Player Pawn Athena Object Reference=@n11) -> (Output_Get:Fort Player Pawn Athena Object Reference)
   Event OnCharacterCustomizationLoaded()
   > n12 = Set Skel Mesh(Skel Mesh=GetPartSkeletalMeshComponent())
   > n13 = Get Dynamic Material Instance(SkeletalMeshComponent=@n12)
   > Set Mat Instance Dynamic(Mat Instance Dynamic=@n13)
   > n14 = GetOrCreateDynamicOverlayMaterial(MeshComponent=@n12, MaterialName=None)
   > Set Overlay Mat Instance Dynamic(Overlay Mat Instance Dynamic=@n14)
   Branch(Condition=Get bIsInFrontEnd())
   Branch(Condition=Get bIsSkyDiving())
   Get bIsTargeting()
   Get bIsTargeting()
  ]}
]
variables [
 FrontEndAlpha {type=number}
 Gravity Override {type=Vector}
 Gravity Override Params {type=GravityOverrideParamsStruct}
 Is Aiming Pistol {type=boolean, default=true}
 Is Aiming Rifle {type=boolean, default=true}
 Local Noise Offset {type=Vector}
 Local Noise Transform Location {type=Vector}
 Mat Instance Dynamic {type=MaterialInstanceDynamic}
 Overlay Mat Instance Dynamic {type=MaterialInstanceDynamic}
 Pistol {type=number}
 Player Pawn {type=FortPlayerPawnAthena}
 R1 Velocity Transform Location {type=Vector}
 R1 Velocity Transform Velocity {type=Vector}
 R2 Velocity Transform Location {type=Vector}
 R2 Velocity Transform Velocity {type=Vector}
 Rifle {type=number}
 Skel Mesh {type=SkeletalMeshComponent}
 Velocity Interp Speed {type=number}
 WPO Pho Wind Intensity {type=number, default=1}
]
