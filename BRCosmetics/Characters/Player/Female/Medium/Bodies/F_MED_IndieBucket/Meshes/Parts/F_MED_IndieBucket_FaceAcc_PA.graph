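# Physics asset for the female medium IndieBucket character's FaceAcc part, set up for two jointed ponytails that swing against the upper body.
# Left and right chains run from head-linked master spheres through a back-hair capsule and seven ponytail capsules, with narrow swing and tight twist limits.
# Capsules on pelvis, spine, neck, head, clavicles and upper arms form the body volume, joined by free-motion constraints.
# Spine_05 adds two long angled capsules to widen that volume.
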PhysicsAsset /BRCosmetics/Characters/Player/Female/Medium/Bodies/F_MED_IndieBucket/Meshes/Parts/F_MED_IndieBucket_FaceAcc_PA
bodies {
 clavicle_l [{shapeType=Capsule, length=5.5, radius=7.5, shapeName=clavicle_l_capsule, center=<7.5 2 0>, rotation=<rot 90 0 0>}]
 clavicle_r [{shapeType=Capsule, length=5.5, radius=7, shapeName=clavicle_r_capsule, center=<-7.5 -2 0>, rotation=<rot 90 0 0>}]
 hair_bk_l_01_dyn [{shapeType=Capsule, length=2, radius=3, shapeName=hair_bk_l_01_dyn_capsule, center=<3 0 1>, rotation=<rot 90 0 0>}]
 hair_bk_l_01_mstr [{shapeType=Sphere, radius=0.505, shapeName=hair_bk_l_01_mstr_sphere}]
 hair_bk_r_01_dyn [{shapeType=Capsule, length=2, radius=3, shapeName=hair_bk_r_01_dyn_capsule, center=<3 0 1>, rotation=<rot 90 0 0>}]
 hair_bk_r_mstr [{shapeType=Sphere, radius=0.505, shapeName=hair_bk_r_mstr_sphere}]
 head [{shapeType=Capsule, length=8, radius=7, shapeName=head_capsule, center=<0 2 0>, rotation=<rot 90 0 0>}]
 neck_01 [{shapeType=Capsule, length=5, radius=5, shapeName=neck_01_capsule, center=<4 0 0>, rotation=<rot 90 0 0>}]
 neck_02 [{shapeType=Capsule, length=5, radius=5, shapeName=neck_02_capsule, center=<4 0 0>, rotation=<rot 90 0 0>}]
 pelvis [{shapeType=Capsule, length=10, radius=10, center=<0 -3.5 0>}]
 ponytail_l_01_dyn [{shapeType=Capsule, length=3.9544, radius=3.5, shapeName=ponytail_l_01_dyn_capsule, center=<4.5 0 2>, rotation=<rot 90 -160.5288 -160.5288>}]
 ponytail_l_02_dyn [{shapeType=Capsule, length=2, radius=4, shapeName=ponytail_l_02_dyn_capsule, center=<3 0 2>, rotation=<rot 90 0 0>}]
 ponytail_l_03_dyn [{shapeType=Capsule, length=2, radius=4, shapeName=ponytail_l_03_dyn_capsule, center=<3 0 2>, rotation=<rot 90 0 0>}]
 ponytail_l_04_dyn [{shapeType=Capsule, length=2, radius=4, shapeName=ponytail_l_04_dyn_capsule, center=<3 0 2>, rotation=<rot 90 0 0>}]
 ponytail_l_05_dyn [{shapeType=Capsule, length=2, radius=4, shapeName=ponytail_l_05_dyn_capsule, center=<2.5 0 1>, rotation=<rot 90 0 0>}]
 ponytail_l_06_dyn [{shapeType=Capsule, length=2, radius=4, shapeName=ponytail_l_06_dyn_capsule, center=<2 0 1>, rotation=<rot 90 0 0>}]
 ponytail_l_07_dyn [{shapeType=Capsule, length=4, radius=4, shapeName=ponytail_l_07_dyn_capsule, center=<6 0 1>, rotation=<rot 90 0 0>}]
 ponytail_r_01_dyn [{shapeType=Capsule, length=2, radius=4, shapeName=ponytail_r_01_dyn_capsule, center=<6 0 1>, rotation=<rot 90 0 0>}]
 ponytail_r_02_dyn [{shapeType=Capsule, length=2, radius=4, shapeName=ponytail_r_02_dyn_capsule, center=<3 0 1.5>, rotation=<rot 90 0 0>}]
 ponytail_r_03_dyn [{shapeType=Capsule, length=2, radius=4, shapeName=ponytail_r_03_dyn_capsule, center=<3 0 2>, rotation=<rot 90 0 0>}]
 ponytail_r_04_dyn [{shapeType=Capsule, length=2, radius=4, shapeName=ponytail_r_04_dyn_capsule, center=<3 0 2>, rotation=<rot 90 0 0>}]
 ponytail_r_05_dyn [{shapeType=Capsule, length=2, radius=4, shapeName=ponytail_r_05_dyn_capsule, center=<2.5 0 2>, rotation=<rot 90 0 0>}]
 ponytail_r_06_dyn [{shapeType=Capsule, length=2, radius=4, shapeName=ponytail_r_06_dyn_capsule, center=<2 0 2>, rotation=<rot 90 0 0>}]
 ponytail_r_07_dyn [{shapeType=Capsule, length=4, radius=4, shapeName=ponytail_r_07_dyn_capsule, center=<6 0 2>, rotation=<rot 90 0 0>}]
 spine_01 [{shapeType=Capsule, length=10, radius=10, shapeName=spine_01_capsule, center=<0 -2.5 0>}]
 spine_02 [{shapeType=Capsule, length=10, radius=8, shapeName=spine_02_capsule}]
 spine_03 [{shapeType=Capsule, length=6, radius=11, shapeName=spine_03_capsule, center=<0 -1 0>}]
 spine_04 [{shapeType=Capsule, length=6, radius=10.5, shapeName=spine_04_capsule, center=<0 -1 0>}]
 spine_05 [
  {shapeType=Capsule, length=10, radius=8, shapeName=spine_05_capsule}
  {shapeType=Capsule, length=24, radius=7, shapeName=spine_05_capsule, center=<18 -6 -25>, rotation=<rot 0 0 40>}
  {shapeType=Capsule, length=24, radius=7, shapeName=spine_05_capsule, center=<18 -5.5 25.5>, rotation=<rot 0 0 -40>}
 ]
 upperarm_l [{shapeType=Capsule, length=17, radius=7, shapeName=upperarm_l_capsule, center=<17 2 -1>, rotation=<rot 90 0 0>}]
 upperarm_r [{shapeType=Capsule, length=16, radius=7, shapeName=upperarm_r_capsule, center=<-17 -2 1>, rotation=<rot 90 0 0>}]
}
constraints [
 {
  bone1Name=spine_01
  bone2Name=pelvis
  swing1LimitDegrees=45
  swing1Motion=Free
  swing2LimitDegrees=45
  swing2Motion=Free
  twistLimitDegrees=45
  twistMotion=Free
 }
 {
  bone1Name=spine_02
  bone2Name=spine_01
  swing1LimitDegrees=45
  swing1Motion=Free
  swing2LimitDegrees=45
  swing2Motion=Free
  twistLimitDegrees=45
  twistMotion=Free
 }
 {
  bone1Name=spine_03
  bone2Name=spine_02
  swing1LimitDegrees=45
  swing1Motion=Free
  swing2LimitDegrees=45
  swing2Motion=Free
  twistLimitDegrees=45
  twistMotion=Free
 }
 {
  bone1Name=spine_04
  bone2Name=spine_03
  swing1LimitDegrees=45
  swing1Motion=Free
  swing2LimitDegrees=45
  swing2Motion=Free
  twistLimitDegrees=45
  twistMotion=Free
 }
 {
  bone1Name=spine_05
  bone2Name=spine_04
  swing1LimitDegrees=45
  swing1Motion=Free
  swing2LimitDegrees=45
  swing2Motion=Free
  twistLimitDegrees=45
  twistMotion=Free
 }
 {
  bone1Name=head
  bone2Name=neck_02
  swing1LimitDegrees=45
  swing1Motion=Free
  swing2LimitDegrees=45
  swing2Motion=Free
  twistLimitDegrees=45
  twistMotion=Free
 }
 {
  bone1Name=hair_bk_l_01_mstr
  bone2Name=head
  swing1LimitDegrees=45
  swing1Motion=Free
  swing2LimitDegrees=45
  swing2Motion=Free
  twistLimitDegrees=45
  twistMotion=Free
 }
 {
  bone1Name=hair_bk_r_mstr
  bone2Name=head
  swing1LimitDegrees=45
  swing1Motion=Free
  swing2LimitDegrees=45
  swing2Motion=Free
  twistLimitDegrees=45
  twistMotion=Free
 }
 {
  bone1Name=clavicle_l
  bone2Name=spine_05
  swing1LimitDegrees=45
  swing1Motion=Free
  swing2LimitDegrees=45
  swing2Motion=Free
  twistLimitDegrees=45
  twistMotion=Free
 }
 {
  bone1Name=upperarm_l
  bone2Name=clavicle_l
  swing1LimitDegrees=45
  swing1Motion=Free
  swing2LimitDegrees=45
  swing2Motion=Free
  twistLimitDegrees=45
  twistMotion=Free
 }
 {
  bone1Name=clavicle_r
  bone2Name=spine_05
  swing1LimitDegrees=45
  swing1Motion=Free
  swing2LimitDegrees=45
  swing2Motion=Free
  twistLimitDegrees=45
  twistMotion=Free
 }
 {
  bone1Name=upperarm_r
  bone2Name=clavicle_r
  swing1LimitDegrees=45
  swing1Motion=Free
  swing2LimitDegrees=45
  swing2Motion=Free
  twistLimitDegrees=45
  twistMotion=Free
 }
 {
  bone1Name=neck_02
  bone2Name=neck_01
  swing1LimitDegrees=45
  swing1Motion=Free
  swing2LimitDegrees=45
  swing2Motion=Free
  twistLimitDegrees=45
  twistMotion=Free
 }
 {
  bone1Name=neck_01
  bone2Name=spine_05
  swing1LimitDegrees=45
  swing1Motion=Free
  swing2LimitDegrees=45
  swing2Motion=Free
  twistLimitDegrees=45
  twistMotion=Free
 }
 {bone1Name=hair_bk_l_01_dyn, bone2Name=hair_bk_l_01_mstr, swing1LimitDegrees=10, swing2LimitDegrees=25, twistLimitDegrees=5}
 {bone1Name=ponytail_l_01_dyn, bone2Name=hair_bk_l_01_dyn, swing1LimitDegrees=15, swing2LimitDegrees=20, twistLimitDegrees=5}
 {bone1Name=ponytail_l_02_dyn, bone2Name=ponytail_l_01_dyn, swing1LimitDegrees=15, swing2LimitDegrees=15, twistLimitDegrees=5}
 {bone1Name=ponytail_l_03_dyn, bone2Name=ponytail_l_02_dyn, swing1LimitDegrees=20, swing2LimitDegrees=30, twistLimitDegrees=5}
 {bone1Name=ponytail_l_04_dyn, bone2Name=ponytail_l_03_dyn, swing1LimitDegrees=20, swing2LimitDegrees=30, twistLimitDegrees=5}
 {bone1Name=ponytail_l_05_dyn, bone2Name=ponytail_l_04_dyn, swing1LimitDegrees=20, swing2LimitDegrees=30, twistLimitDegrees=5}
 {bone1Name=ponytail_l_06_dyn, bone2Name=ponytail_l_05_dyn, swing1LimitDegrees=20, swing2LimitDegrees=30, twistLimitDegrees=5}
 {bone1Name=ponytail_l_07_dyn, bone2Name=ponytail_l_06_dyn, swing1LimitDegrees=20, swing2LimitDegrees=30, twistLimitDegrees=5}
 {bone1Name=hair_bk_r_01_dyn, bone2Name=hair_bk_r_mstr, swing1LimitDegrees=10, swing2LimitDegrees=25, twistLimitDegrees=1}
 {bone1Name=ponytail_r_01_dyn, bone2Name=hair_bk_r_01_dyn, swing1LimitDegrees=15, swing2LimitDegrees=20, twistLimitDegrees=5}
 {bone1Name=ponytail_r_02_dyn, bone2Name=ponytail_r_01_dyn, swing1LimitDegrees=15, swing2LimitDegrees=15, twistLimitDegrees=5}
 {bone1Name=ponytail_r_03_dyn, bone2Name=ponytail_r_02_dyn, swing1LimitDegrees=20, swing2LimitDegrees=30, twistLimitDegrees=5}
 {bone1Name=ponytail_r_04_dyn, bone2Name=ponytail_r_03_dyn, swing1LimitDegrees=20, swing2LimitDegrees=30, twistLimitDegrees=5}
 {bone1Name=ponytail_r_05_dyn, bone2Name=ponytail_r_04_dyn, swing1LimitDegrees=20, swing2LimitDegrees=30, twistLimitDegrees=5}
 {bone1Name=ponytail_r_06_dyn, bone2Name=ponytail_r_05_dyn, swing1LimitDegrees=20, swing2LimitDegrees=30, twistLimitDegrees=5}
 {bone1Name=ponytail_r_07_dyn, bone2Name=ponytail_r_06_dyn, swing1LimitDegrees=20, swing2LimitDegrees=30, twistLimitDegrees=5}
]
