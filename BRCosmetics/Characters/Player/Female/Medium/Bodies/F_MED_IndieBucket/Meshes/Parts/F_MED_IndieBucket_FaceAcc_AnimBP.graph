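# Post-process style AnimBlueprint for the IndieBucket female face accessory mesh, driving secondary motion on bangs, back hair, ponytails and earring.
# Anim Dynamics springs move bangs, and a cached 'Anim Dynamics' pose feeds branches mixing RigidBody simulation, weighted multi-bone constraints and additive bone corrections.
# Bool blends select branches for backbling, standing sprint and front end.
# The backbling branch pins hair to torso bones without RigidBody; front end uses its own physics asset.
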AnimBlueprint /BRCosmetics/Characters/Player/Female/Medium/Bodies/F_MED_IndieBucket/Meshes/Parts/F_MED_IndieBucket_FaceAcc_AnimBP
target_skeleton=&/F_MED_IndieBucket_FaceAcc_Skeleton
graphs [
 AnimGraph {nodes [
   Anim Dynamics - Bone: bang_right_a_01_dyn(Alpha=0.5, ExternalForce="0,0,0") {
    angularDampingOverride=0.9999
    angularSpringConstant=250
    bAngularSpring=true
    bGravityOverrideInSimSpace=true
    bOverrideAngularDamping=true
    simulationSpace=RootRelative
    alphaBoolBlend {blendOption=Linear}
    boundBone {boneName=bang_right_a_01_dyn}
    componentAppliedLinearAccClamp=<5000 5000 5000>
    componentLinearAccScale=<1 1 1>
    componentLinearVelScale=<1 1 1>
    physicsBodyDefinitions [{
      sphereCollisionRadius=0
      boundBone {boneName=bang_right_a_01_dyn}
      boxExtents=<20 20 10>
      constraintSetup {angularLimitsMax=<5 20 20>, angularLimitsMin=<-5 10 5>, angularTarget=<0 0 0>}
      localJointOffset=<1 0 -0.5>
     }]
   }
   Anim Dynamics - Bone: bang_right_b_01_dyn(Alpha=0.5, ExternalForce="0,0,0") {
    angularDampingOverride=0.9999
    angularSpringConstant=250
    bAngularSpring=true
    bGravityOverrideInSimSpace=true
    bOverrideAngularDamping=true
    simulationSpace=RootRelative
    alphaBoolBlend {blendOption=Linear}
    boundBone {boneName=bang_right_b_01_dyn}
    componentAppliedLinearAccClamp=<5000 5000 5000>
    componentLinearAccScale=<1 1 1>
    componentLinearVelScale=<1 1 1>
    physicsBodyDefinitions [{
      sphereCollisionRadius=0
      boundBone {boneName=bang_right_b_01_dyn}
      boxExtents=<20 20 10>
      constraintSetup {angularLimitsMax=<5 15 20>, angularLimitsMin=<-5 0 0>, angularTarget=<0 0 0>}
      localJointOffset=<1 0 -0.5>
     }]
   }
   Anim Dynamics - Bone: bang_right_c_01_dyn(Alpha=0.5, ExternalForce="0,0,0") {
    angularDampingOverride=0.9999
    angularSpringConstant=250
    bAngularSpring=true
    bGravityOverrideInSimSpace=true
    bOverrideAngularDamping=true
    simulationSpace=RootRelative
    alphaBoolBlend {blendOption=Linear}
    boundBone {boneName=bang_right_c_01_dyn}
    componentAppliedLinearAccClamp=<5000 5000 5000>
    componentLinearAccScale=<1 1 1>
    componentLinearVelScale=<1 1 1>
    physicsBodyDefinitions [{
      sphereCollisionRadius=0
      boundBone {boneName=bang_right_c_01_dyn}
      boxExtents=<20 20 10>
      constraintSetup {angularLimitsMax=<5 0 0>, angularLimitsMin=<-5 -15 -15>, angularTarget=<0 0 0>}
      localJointOffset=<1 0 -0.5>
     }]
   }
   Anim Dynamics - Bone: bang_right_d_01_dyn(Alpha=0.5, ExternalForce="0,0,0") {
    angularDampingOverride=0.9999
    angularSpringConstant=250
    bAngularSpring=true
    bGravityOverrideInSimSpace=true
    bOverrideAngularDamping=true
    simulationSpace=RootRelative
    alphaBoolBlend {blendOption=Linear}
    boundBone {boneName=bang_right_d_01_dyn}
    componentAppliedLinearAccClamp=<5000 5000 5000>
    componentLinearAccScale=<1 1 1>
    componentLinearVelScale=<1 1 1>
    physicsBodyDefinitions [{
      sphereCollisionRadius=0
      boundBone {boneName=bang_right_d_01_dyn}
      boxExtents=<20 20 10>
      constraintSetup {angularLimitsMax=<5 0 0>, angularLimitsMin=<-5 0 -15>, angularTarget=<0 0 0>}
      localJointOffset=<1 0 -0.5>
     }]
   }
   n1 = Get bIsInFrontEnd()
   "Transform (Modify) Bone - Bone: dyn_l_earring_mstr"!(ComponentPose, bAlphaBoolEnabled:bool=@n1, Translation="0,0,0", Rotation="0,20,0", Scale="1,1,1") -> (Pose) {
    alphaInputType=Bool
    rotationMode=BMM_Additive
    rotationSpace=BCS_BoneSpace
    translationMode=BMM_Additive
    translationSpace=BCS_BoneSpace
    alphaBoolBlend {blendOption=Linear}
    boneToModify {boneName=dyn_l_earring_mstr}
    rotation {yaw=20}
   }
   n2 = RigidBody!(ComponentPose=Local To Component(LocalPose=Use cached pose 'Anim Dynamics'()), Alpha, OverrideWorldGravity:Vector="0,0,-980", ExternalForce) -> (Pose) {
    bClampLinearTranslationLimitToRefPose=true
    bOverrideWorldGravity=true
    overridePhysicsAsset=&/F_MED_IndieBucket_FaceAcc_PA
    simulationSpace=BaseBoneSpace
    alphaBoolBlend {blendOption=Linear}
    baseBoneRef {boneName=spine_05}
    componentAppliedLinearAccClamp=<3000 3000 3000>
    componentLinearAccScale=<0.5 0.3 0.3>
    componentLinearVelScale=<0.25 0.15 0.15>
    simSpaceSettings {
     maxAngularAcceleration=1000
     maxAngularVelocity=3
     maxLinearAcceleration=700
     maxLinearVelocity=500
     velocityScaleZ=5
     worldAlpha=1
     externalLinearDragV=<0.2 0.2>
    }
   }
   n3 = "Transform (Modify) Bone - Bone: hair_bk_l_01_mstr"!(ComponentPose=@n2, bAlphaBoolEnabled:bool=Get crouch moving() -> (crouch moving:bool), Translation="0,0,0", Rotation="10,0,0", Scale="1,1,1") -> (Pose) {
    alphaInputType=Bool
    rotationMode=BMM_Additive
    rotationSpace=BCS_BoneSpace
    translationMode=BMM_Additive
    translationSpace=BCS_BoneSpace
    alphaBoolBlend {blendInTime=0.1, blendOption=Linear, blendOutTime=0.1}
    boneToModify {boneName=hair_bk_l_01_mstr}
    rotation {pitch=10}
   }
   n4 = "Transform (Modify) Bone - Bone: hair_bk_l_01_mstr"!(ComponentPose=@n3, Alpha:float=Get head left() -> (head left:double), Translation="0,0,0", Rotation="-10,0,0", Scale="1,1,1") -> (Pose) {
    rotationMode=BMM_Additive
    rotationSpace=BCS_BoneSpace
    translationMode=BMM_Additive
    translationSpace=BCS_BoneSpace
    alphaBoolBlend {blendInTime=0.1, blendOption=Linear, blendOutTime=0.1}
    alphaScaleBiasClamp {bInterpResult=true}
    boneToModify {boneName=hair_bk_l_01_mstr}
    rotation {pitch=-10}
   }
   n5 = "Transform (Modify) Bone - Bone: bang_left_a_front_end_correction_dyn"(ComponentPose=@n4, Alpha=1, Translation="0,0,-7", Rotation="0,0,0", Scale="1,1,1") {
    rotationMode=BMM_Additive
    rotationSpace=BCS_BoneSpace
    translationMode=BMM_Additive
    translationSpace=BCS_BoneSpace
    alphaBoolBlend {blendInTime=0.1, blendOption=Linear, blendOutTime=0.1}
    alphaScaleBiasClamp {bInterpResult=true}
    boneToModify {boneName=bang_left_a_front_end_correction_dyn}
    translation {z=-7}
   }
   n6 = Component To Local(ComponentPose=@n5)
   n7 = "Constraint - hair_bk_r_mstr (P)"(ComponentPose=Local To Component(LocalPose=Use cached pose 'Anim Dynamics'()), Alpha=0.4, ConstraintWeights_0=1) {
    alphaBoolBlend {blendOption=Linear}
    boneToModify {boneName=hair_bk_r_mstr}
    constraintSetup [{targetBone {boneName=spine_05}}]
    constraintWeights [1]
   }
   n8 = "Constraint - hair_bk_l_01_mstr (P)"(ComponentPose=@n7, Alpha=0.1, ConstraintWeights_0=1) {
    alphaBoolBlend {blendOption=Linear}
    boneToModify {boneName=hair_bk_l_01_mstr}
    constraintSetup [{targetBone {boneName=spine_05}}]
    constraintWeights [1]
   }
   n9 = RigidBody(ComponentPose=@n8, OverrideWorldGravity="0,0,-800") {
    bClampLinearTranslationLimitToRefPose=true
    bOverrideWorldGravity=true
    overridePhysicsAsset=&/F_MED_frontend_IndieBucket_FaceAcc_PA1
    simulationSpace=BaseBoneSpace
    alphaBoolBlend {blendOption=Linear}
    baseBoneRef {boneName=spine_05}
    componentAppliedLinearAccClamp=<2500 2500 2500>
    componentLinearAccScale=<0.3 0.3 0.3>
    componentLinearVelScale=<0.3 0.3 0.3>
    overrideWorldGravity {z=-800}
   }
   n10 = Component To Local(ComponentPose=@n9)
   n11 = Get bIsInFrontEnd()
   n12 = "Constraint - hair_bk_l_01_mstr (PP)"!(
    ComponentPose=Local To Component(LocalPose=Use cached pose 'Anim Dynamics'())
    Alpha=0.75
    ConstraintWeights_0=0.9995
    ConstraintWeights_1:float=0.0005
   ) -> (Pose) {
    alphaBoolBlend {blendOption=Linear}
    alphaScaleBiasClamp {bInterpResult=true}
    boneToModify {boneName=hair_bk_l_01_mstr}
    constraintSetup [{targetBone {boneName=spine_05}}, {targetBone {boneName=neck_01}}]
    constraintWeights [1,1]
   }
   n13 = "Constraint - ponytail_l_01_mstr (PPPPPPPP)"!(
    ComponentPose=@n12
    Alpha=0.75
    ConstraintWeights_0=0
    ConstraintWeights_1:float=0.0004
    ConstraintWeights_2:float=0.6951
    ConstraintWeights_3:float=0
    ConstraintWeights_4:float=0.2589
    ConstraintWeights_5:float=0
    ConstraintWeights_6:float=0
    ConstraintWeights_7:float=0.0457
   ) -> (Pose) {
    alphaBoolBlend {blendOption=Linear}
    alphaScaleBiasClamp {bInterpResult=true}
    boneToModify {boneName=ponytail_l_01_mstr}
    constraintSetup [
     {targetBone {boneName=spine_03}}
     {targetBone {boneName=spine_04}}
     {targetBone {boneName=spine_05}}
     {targetBone {boneName=neck_01}}
     {targetBone {boneName=clavicle_l}}
     {targetBone {boneName=upperarm_l}}
     {targetBone {boneName=upperarm_twist_01_l}}
     {targetBone {boneName=clavicle_r}}
    ]
    constraintWeights [1,1,1,1,1,1,1,1]
   }
   n14 = "Constraint - ponytail_l_02_dyn (PPPPPPPP)"!(
    ComponentPose=@n13
    Alpha=0.75
    ConstraintWeights_0=0.0001
    ConstraintWeights_1:float=0.0001
    ConstraintWeights_2:float=0.0054
    ConstraintWeights_3:float=0.5939
    ConstraintWeights_4:float=0.3168
    ConstraintWeights_5:float=0.0001
    ConstraintWeights_6:float=0.0006
    ConstraintWeights_7:float=0.0829
   ) -> (Pose) {
    alphaBoolBlend {blendOption=Linear}
    alphaScaleBiasClamp {bInterpResult=true}
    boneToModify {boneName=ponytail_l_02_dyn}
    constraintSetup [
     {targetBone {boneName=spine_02}}
     {targetBone {boneName=spine_03}}
     {targetBone {boneName=spine_04}}
     {targetBone {boneName=spine_05}}
     {targetBone {boneName=clavicle_l}}
     {targetBone {boneName=upperarm_l}}
     {targetBone {boneName=upperarm_twist_01_l}}
     {targetBone {boneName=clavicle_r}}
    ]
    constraintWeights [1,1,1,1,1,1,1,1]
   }
   n15 = "Constraint - hair_bk_r_mstr (PP)"!(ComponentPose=@n14, Alpha=0.75, ConstraintWeights_0=0.9995, ConstraintWeights_1:float=0.0005) -> (Pose) {
    alphaBoolBlend {blendOption=Linear}
    alphaScaleBiasClamp {bInterpResult=true}
    boneToModify {boneName=hair_bk_r_mstr}
    constraintSetup [{targetBone {boneName=spine_05}}, {targetBone {boneName=neck_01}}]
    constraintWeights [1,1]
   }
   n16 = "Constraint - ponytail_r_01_mstr (PPPPP)"!(
    ComponentPose=@n15
    Alpha=0.75
    ConstraintWeights_0=0.0011
    ConstraintWeights_1:float=0.5621
    ConstraintWeights_2:float=0.027
    ConstraintWeights_3:float=0.4098
    ConstraintWeights_4:float=0
   ) -> (Pose) {
    alphaBoolBlend {blendOption=Linear}
    alphaScaleBiasClamp {bInterpResult=true}
    boneToModify {boneName=ponytail_r_01_mstr}
    constraintSetup [
     {targetBone {boneName=spine_04}}
     {targetBone {boneName=spine_05}}
     {targetBone {boneName=clavicle_l}}
     {targetBone {boneName=clavicle_r}}
     {targetBone {boneName=upperarm_twist_01_r}}
    ]
    constraintWeights [1,1,1,1,1]
   }
   n17 = "Constraint - ponytail_r_02_dyn (PPPPP)"!(
    ComponentPose=@n16
    Alpha=0.75
    ConstraintWeights_0=0.01
    ConstraintWeights_1:float=0.5003
    ConstraintWeights_2:float=0.0417
    ConstraintWeights_3:float=0.447
    ConstraintWeights_4:float=0.0009
   ) -> (Pose) {
    alphaBoolBlend {blendOption=Linear}
    alphaScaleBiasClamp {bInterpResult=true}
    boneToModify {boneName=ponytail_r_02_dyn}
    constraintSetup [
     {targetBone {boneName=spine_04}}
     {targetBone {boneName=spine_05}}
     {targetBone {boneName=clavicle_l}}
     {targetBone {boneName=clavicle_r}}
     {targetBone {boneName=upperarm_twist_01_r}}
    ]
    constraintWeights [1,1,1,1,1]
   }
   n18 = "Transform (Modify) Bone - Bone: hair_bk_l_01_mstr"(ComponentPose=@n17, Alpha=0.5, Translation="0,-1,0", Rotation="-5,-15,0", Scale="1,1,1") {
    rotationMode=BMM_Additive
    rotationSpace=BCS_BoneSpace
    translationMode=BMM_Additive
    translationSpace=BCS_BoneSpace
    alphaBoolBlend {blendOption=Linear}
    boneToModify {boneName=hair_bk_l_01_mstr}
    rotation {pitch=-5, yaw=-15}
    translation {y=-1}
   }
   n19 = "Transform (Modify) Bone - Bone: hair_bk_r_01_dyn"(ComponentPose=@n18, Alpha=0.5, Translation="0,1,0", Rotation="-5,15,0", Scale="1,1,1") {
    rotationMode=BMM_Additive
    rotationSpace=BCS_BoneSpace
    translationMode=BMM_Additive
    translationSpace=BCS_BoneSpace
    alphaBoolBlend {blendOption=Linear}
    boneToModify {boneName=hair_bk_r_01_dyn}
    rotation {pitch=-5, yaw=15}
    translation {y=1}
   }
   n20 = "Transform (Modify) Bone - Bone: ponytail_l_mid_mstr"(ComponentPose=@n19, Alpha=0.5, Translation="0,0,0", Rotation="-15,0,0", Scale="1,1,1") {
    rotationMode=BMM_Additive
    rotationSpace=BCS_BoneSpace
    translationMode=BMM_Additive
    translationSpace=BCS_BoneSpace
    alphaBoolBlend {blendOption=Linear}
    boneToModify {boneName=ponytail_l_mid_mstr}
    rotation {pitch=-15}
   }
   n21 = "Transform (Modify) Bone - Bone: ponytail_r_mid_mstr"(ComponentPose=@n20, Alpha=0.5, Translation="0,0,0", Rotation="-15,0,0", Scale="1,1,1") {
    rotationMode=BMM_Additive
    rotationSpace=BCS_BoneSpace
    translationMode=BMM_Additive
    translationSpace=BCS_BoneSpace
    alphaBoolBlend {blendOption=Linear}
    boneToModify {boneName=ponytail_r_mid_mstr}
    rotation {pitch=-15}
   }
   n22 = RigidBody(ComponentPose=@n21, OverrideWorldGravity="0,0,-980") {
    bClampLinearTranslationLimitToRefPose=true
    bOverrideWorldGravity=true
    overridePhysicsAsset=&F_MED_IndieBucket_FaceAcc_PA
    simulationSpace=BaseBoneSpace
    alphaBoolBlend {blendOption=Linear}
    baseBoneRef {boneName=spine_05}
    componentAppliedLinearAccClamp=<3000 3000 3000>
    componentLinearAccScale=<0.5 0.3 0.3>
    componentLinearVelScale=<0.25 0.15 0.15>
    simSpaceSettings {
     maxAngularAcceleration=1000
     maxAngularVelocity=3
     maxLinearAcceleration=700
     maxLinearVelocity=500
     velocityScaleZ=5
     worldAlpha=1
     externalLinearDragV=<0.2 0.2>
    }
   }
   n23 = "Constraint - hair_bk_l_01_mstr (PP)"(
    ComponentPose=Local To Component(LocalPose=Use cached pose 'Anim Dynamics'())
    Alpha=1
    ConstraintWeights_0=0.9995
    ConstraintWeights_1=0.0005
   ) {
    alphaBoolBlend {blendOption=Linear}
    boneToModify {boneName=hair_bk_l_01_mstr}
    constraintSetup [{targetBone {boneName=spine_05}}, {targetBone {boneName=neck_01}}]
    constraintWeights [1,1]
   }
   n24 = "Constraint - ponytail_l_01_mstr (PPPPPPPP)"(
    ComponentPose=@n23
    Alpha=1
    ConstraintWeights_0=0
    ConstraintWeights_1=0.0004
    ConstraintWeights_2=0.6951
    ConstraintWeights_3=0
    ConstraintWeights_4=0.2589
    ConstraintWeights_5=0
    ConstraintWeights_6=0
    ConstraintWeights_7=0.0457
   ) {
    alphaBoolBlend {blendOption=Linear}
    boneToModify {boneName=ponytail_l_01_mstr}
    constraintSetup [
     {targetBone {boneName=spine_03}}
     {targetBone {boneName=spine_04}}
     {targetBone {boneName=spine_05}}
     {targetBone {boneName=neck_01}}
     {targetBone {boneName=clavicle_l}}
     {targetBone {boneName=upperarm_l}}
     {targetBone {boneName=upperarm_twist_01_l}}
     {targetBone {boneName=clavicle_r}}
    ]
    constraintWeights [1,1,1,1,1,1,1,1]
   }
   n25 = "Constraint - ponytail_l_02_dyn (PPPPPPPP)"(
    ComponentPose=@n24
    Alpha=1
    ConstraintWeights_0=0.0001
    ConstraintWeights_1=0.0001
    ConstraintWeights_2=0.0054
    ConstraintWeights_3=0.5939
    ConstraintWeights_4=0.3168
    ConstraintWeights_5=0.0001
    ConstraintWeights_6=0.0006
    ConstraintWeights_7=0.0829
   ) {
    alphaBoolBlend {blendOption=Linear}
    boneToModify {boneName=ponytail_l_02_dyn}
    constraintSetup [
     {targetBone {boneName=spine_02}}
     {targetBone {boneName=spine_03}}
     {targetBone {boneName=spine_04}}
     {targetBone {boneName=spine_05}}
     {targetBone {boneName=clavicle_l}}
     {targetBone {boneName=upperarm_l}}
     {targetBone {boneName=upperarm_twist_01_l}}
     {targetBone {boneName=clavicle_r}}
    ]
    constraintWeights [1,1,1,1,1,1,1,1]
   }
   n26 = "Constraint - ponytail_l_mid_mstr (PP)"!(ComponentPose=@n25, Alpha=1, ConstraintWeights_0=1, ConstraintWeights_1:float=0) -> (Pose) {
    alphaBoolBlend {blendOption=Linear}
    boneToModify {boneName=ponytail_l_mid_mstr}
    constraintSetup [{targetBone {boneName=spine_05}}, {targetBone {boneName=neck_01}}]
    constraintWeights [1,1]
   }
   n27 = "Constraint - ponytail_l_04_dyn (PPPPPPPP)"!(
    ComponentPose=@n26
    Alpha=0
    ConstraintWeights_0=0.0087
    ConstraintWeights_1:float=0.0368
    ConstraintWeights_2:float=0.0376
    ConstraintWeights_3:float=0.259
    ConstraintWeights_4:float=0.3554
    ConstraintWeights_5:float=0.2211
    ConstraintWeights_6:float=0.011
    ConstraintWeights_7:float=0.0704
   ) -> (Pose) {
    alphaBoolBlend {blendOption=Linear}
    boneToModify {boneName=ponytail_l_04_dyn}
    constraintSetup [
     {targetBone {boneName=spine_01}}
     {targetBone {boneName=spine_02}}
     {targetBone {boneName=spine_03}}
     {targetBone {boneName=spine_04}}
     {targetBone {boneName=spine_05}}
     {targetBone {boneName=clavicle_l}}
     {targetBone {boneName=upperarm_twist_01_l}}
     {targetBone {boneName=clavicle_r}}
    ]
    constraintWeights [1,1,1,1,1,1,1,1]
   }
   n28 = "Constraint - ponytail_l_05_dyn (PPPPPPPP)"!(
    ComponentPose=@n27
    Alpha=0
    ConstraintWeights_0=0.0346
    ConstraintWeights_1:float=0.1245
    ConstraintWeights_2:float=0.0958
    ConstraintWeights_3:float=0.3398
    ConstraintWeights_4:float=0.1982
    ConstraintWeights_5:float=0.1456
    ConstraintWeights_6:float=0.019
    ConstraintWeights_7:float=0.0426
   ) -> (Pose) {
    alphaBoolBlend {blendOption=Linear}
    boneToModify {boneName=ponytail_l_05_dyn}
    constraintSetup [
     {targetBone {boneName=spine_01}}
     {targetBone {boneName=spine_02}}
     {targetBone {boneName=spine_03}}
     {targetBone {boneName=spine_04}}
     {targetBone {boneName=spine_05}}
     {targetBone {boneName=clavicle_l}}
     {targetBone {boneName=upperarm_twist_01_l}}
     {targetBone {boneName=clavicle_r}}
    ]
    constraintWeights [1,1,1,1,1,1,1,1]
   }
   n29 = "Constraint - ponytail_l_06_dyn (PPPPPPPP)"!(
    ComponentPose=@n28
    Alpha=0
    ConstraintWeights_0=0.0512
    ConstraintWeights_1:float=0.0731
    ConstraintWeights_2:float=0.2901
    ConstraintWeights_3:float=0.1391
    ConstraintWeights_4:float=0.288
    ConstraintWeights_5:float=0.0817
    ConstraintWeights_6:float=0.065
    ConstraintWeights_7:float=0.012
   ) -> (Pose) {
    alphaBoolBlend {blendOption=Linear}
    boneToModify {boneName=ponytail_l_06_dyn}
    constraintSetup [
     {targetBone {boneName=pelvis}}
     {targetBone {boneName=spine_01}}
     {targetBone {boneName=spine_02}}
     {targetBone {boneName=spine_03}}
     {targetBone {boneName=spine_04}}
     {targetBone {boneName=spine_05}}
     {targetBone {boneName=clavicle_l}}
     {targetBone {boneName=clavicle_r}}
    ]
    constraintWeights [1,1,1,1,1,1,1,1]
   }
   n30 = "Constraint - ponytail_l_07_dyn (PPPPPPPP)"!(
    ComponentPose=@n29
    Alpha=0
    ConstraintWeights_0=0.1623
    ConstraintWeights_1:float=0.0924
    ConstraintWeights_2:float=0.3629
    ConstraintWeights_3:float=0.1187
    ConstraintWeights_4:float=0.1958
    ConstraintWeights_5:float=0.036
    ConstraintWeights_6:float=0.0298
    ConstraintWeights_7:float=0.0022
   ) -> (Pose) {
    alphaBoolBlend {blendOption=Linear}
    boneToModify {boneName=ponytail_l_07_dyn}
    constraintSetup [
     {targetBone {boneName=pelvis}}
     {targetBone {boneName=spine_01}}
     {targetBone {boneName=spine_02}}
     {targetBone {boneName=spine_03}}
     {targetBone {boneName=spine_04}}
     {targetBone {boneName=spine_05}}
     {targetBone {boneName=clavicle_l}}
     {targetBone {boneName=clavicle_r}}
    ]
    constraintWeights [1,1,1,1,1,1,1,1]
   }
   n31 = "Constraint - hair_bk_r_mstr (PP)"(ComponentPose=@n30, Alpha=1, ConstraintWeights_0=0.9995, ConstraintWeights_1=0.0005) {
    alphaBoolBlend {blendOption=Linear}
    boneToModify {boneName=hair_bk_r_mstr}
    constraintSetup [{targetBone {boneName=spine_05}}, {targetBone {boneName=neck_01}}]
    constraintWeights [1,1]
   }
   n32 = "Constraint - ponytail_r_01_mstr (PPPPP)"(
    ComponentPose=@n31
    Alpha=1
    ConstraintWeights_0=0.0011
    ConstraintWeights_1=0.5621
    ConstraintWeights_2=0.027
    ConstraintWeights_3=0.4098
    ConstraintWeights_4=0
   ) {
    alphaBoolBlend {blendOption=Linear}
    boneToModify {boneName=ponytail_r_01_mstr}
    constraintSetup [
     {targetBone {boneName=spine_04}}
     {targetBone {boneName=spine_05}}
     {targetBone {boneName=clavicle_l}}
     {targetBone {boneName=clavicle_r}}
     {targetBone {boneName=upperarm_twist_01_r}}
    ]
    constraintWeights [1,1,1,1,1]
   }
   n33 = "Constraint - ponytail_r_02_dyn (PPPPP)"(
    ComponentPose=@n32
    Alpha=1
    ConstraintWeights_0=0.01
    ConstraintWeights_1=0.5003
    ConstraintWeights_2=0.0417
    ConstraintWeights_3=0.447
    ConstraintWeights_4=0.0009
   ) {
    alphaBoolBlend {blendOption=Linear}
    boneToModify {boneName=ponytail_r_02_dyn}
    constraintSetup [
     {targetBone {boneName=spine_04}}
     {targetBone {boneName=spine_05}}
     {targetBone {boneName=clavicle_l}}
     {targetBone {boneName=clavicle_r}}
     {targetBone {boneName=upperarm_twist_01_r}}
    ]
    constraintWeights [1,1,1,1,1]
   }
   n34 = "Constraint - ponytail_r_mid_mstr (PPPPPPPP)"!(
    ComponentPose=@n33
    Alpha=1
    ConstraintWeights_0=0.0106
    ConstraintWeights_1:float=0.0112
    ConstraintWeights_2:float=0.1232
    ConstraintWeights_3:float=0.4082
    ConstraintWeights_4:float=0.0752
    ConstraintWeights_5:float=0.3382
    ConstraintWeights_6:float=0.0099
    ConstraintWeights_7:float=0.0235
   ) -> (Pose) {
    alphaBoolBlend {blendOption=Linear}
    boneToModify {boneName=ponytail_r_mid_mstr}
    constraintSetup [
     {targetBone {boneName=spine_02}}
     {targetBone {boneName=spine_03}}
     {targetBone {boneName=spine_04}}
     {targetBone {boneName=spine_05}}
     {targetBone {boneName=clavicle_l}}
     {targetBone {boneName=clavicle_r}}
     {targetBone {boneName=upperarm_r}}
     {targetBone {boneName=upperarm_twist_01_r}}
    ]
    constraintWeights [1,1,1,1,1,1,1,1]
   }
   n35 = "Constraint - ponytail_r_04_dyn (PPPPPPPP)"!(
    ComponentPose=@n34
    Alpha=0
    ConstraintWeights_0=0.0126
    ConstraintWeights_1:float=0.0701
    ConstraintWeights_2:float=0.055
    ConstraintWeights_3:float=0.2816
    ConstraintWeights_4:float=0.2387
    ConstraintWeights_5:float=0.0325
    ConstraintWeights_6:float=0.2631
    ConstraintWeights_7:float=0.0463
   ) -> (Pose) {
    alphaBoolBlend {blendOption=Linear}
    boneToModify {boneName=ponytail_r_04_dyn}
    constraintSetup [
     {targetBone {boneName=spine_01}}
     {targetBone {boneName=spine_02}}
     {targetBone {boneName=spine_03}}
     {targetBone {boneName=spine_04}}
     {targetBone {boneName=spine_05}}
     {targetBone {boneName=clavicle_l}}
     {targetBone {boneName=clavicle_r}}
     {targetBone {boneName=upperarm_twist_01_r}}
    ]
    constraintWeights [1,1,1,1,1,1,1,1]
   }
   n36 = "Constraint - ponytail_r_05_dyn (PPPPPPPP)"!(
    ComponentPose=@n35
    Alpha=0
    ConstraintWeights_0=0.0328
    ConstraintWeights_1:float=0.1297
    ConstraintWeights_2:float=0.0906
    ConstraintWeights_3:float=0.3226
    ConstraintWeights_4:float=0.1942
    ConstraintWeights_5:float=0.0439
    ConstraintWeights_6:float=0.1612
    ConstraintWeights_7:float=0.025
   ) -> (Pose) {
    alphaBoolBlend {blendOption=Linear}
    boneToModify {boneName=ponytail_r_05_dyn}
    constraintSetup [
     {targetBone {boneName=spine_01}}
     {targetBone {boneName=spine_02}}
     {targetBone {boneName=spine_03}}
     {targetBone {boneName=spine_04}}
     {targetBone {boneName=spine_05}}
     {targetBone {boneName=clavicle_l}}
     {targetBone {boneName=clavicle_r}}
     {targetBone {boneName=upperarm_twist_01_r}}
    ]
    constraintWeights [1,1,1,1,1,1,1,1]
   }
   n37 = "Constraint - ponytail_r_06_dyn (PPPPPPPP)"!(
    ComponentPose=@n36
    Alpha=0
    ConstraintWeights_0=0.0446
    ConstraintWeights_1:float=0.0706
    ConstraintWeights_2:float=0.2895
    ConstraintWeights_3:float=0.1379
    ConstraintWeights_4:float=0.2945
    ConstraintWeights_5:float=0.0828
    ConstraintWeights_6:float=0.0233
    ConstraintWeights_7:float=0.0568
   ) -> (Pose) {
    alphaBoolBlend {blendOption=Linear}
    boneToModify {boneName=ponytail_r_06_dyn}
    constraintSetup [
     {targetBone {boneName=pelvis}}
     {targetBone {boneName=spine_01}}
     {targetBone {boneName=spine_02}}
     {targetBone {boneName=spine_03}}
     {targetBone {boneName=spine_04}}
     {targetBone {boneName=spine_05}}
     {targetBone {boneName=clavicle_l}}
     {targetBone {boneName=clavicle_r}}
    ]
    constraintWeights [1,1,1,1,1,1,1,1]
   }
   n38 = "Constraint - ponytail_r_07_dyn (PPPPPPPP)"!(
    ComponentPose=@n37
    Alpha=0
    ConstraintWeights_0=0.1944
    ConstraintWeights_1:float=0.0983
    ConstraintWeights_2:float=0.3985
    ConstraintWeights_3:float=0.1102
    ConstraintWeights_4:float=0.1583
    ConstraintWeights_5:float=0.0214
    ConstraintWeights_6:float=0.0039
    ConstraintWeights_7:float=0.0151
   ) -> (Pose) {
    alphaBoolBlend {blendOption=Linear}
    boneToModify {boneName=ponytail_r_07_dyn}
    constraintSetup [
     {targetBone {boneName=pelvis}}
     {targetBone {boneName=spine_01}}
     {targetBone {boneName=spine_02}}
     {targetBone {boneName=spine_03}}
     {targetBone {boneName=spine_04}}
     {targetBone {boneName=spine_05}}
     {targetBone {boneName=clavicle_l}}
     {targetBone {boneName=clavicle_r}}
    ]
    constraintWeights [1,1,1,1,1,1,1,1]
   }
   n39 = "Transform (Modify) Bone - Bone: hair_bk_l_01_mstr"(ComponentPose=@n38, Alpha=1, Translation="2.5,-1,0", Rotation="-5,-15,0", Scale="1,1,1") {
    rotationMode=BMM_Additive
    rotationSpace=BCS_BoneSpace
    translationMode=BMM_Additive
    translationSpace=BCS_BoneSpace
    alphaBoolBlend {blendOption=Linear}
    boneToModify {boneName=hair_bk_l_01_mstr}
    rotation {pitch=-5, yaw=-15}
    translation=<2.5 -1>
   }
   n40 = "Transform (Modify) Bone - Bone: hair_bk_r_01_dyn"(ComponentPose=@n39, Alpha=1, Translation="2.5,1,0", Rotation="-5,15,0", Scale="1,1,1") {
    rotationMode=BMM_Additive
    rotationSpace=BCS_BoneSpace
    translationMode=BMM_Additive
    translationSpace=BCS_BoneSpace
    alphaBoolBlend {blendOption=Linear}
    boneToModify {boneName=hair_bk_r_01_dyn}
    rotation {pitch=-5, yaw=15}
    translation=<2.5 1>
   }
   n41 = "Transform (Modify) Bone - Bone: ponytail_l_mid_mstr"(ComponentPose=@n40, Alpha=1, Translation="0,0,0", Rotation="-15,0,0", Scale="1,1,1") {
    rotationMode=BMM_Additive
    rotationSpace=BCS_BoneSpace
    translationMode=BMM_Additive
    translationSpace=BCS_BoneSpace
    alphaBoolBlend {blendOption=Linear}
    boneToModify {boneName=ponytail_l_mid_mstr}
    rotation {pitch=-15}
   }
   n42 = "Transform (Modify) Bone - Bone: ponytail_r_mid_mstr"(ComponentPose=@n41, Alpha=1, Translation="0,0,0", Rotation="-15,0,0", Scale="1,1,1") {
    rotationMode=BMM_Additive
    rotationSpace=BCS_BoneSpace
    translationMode=BMM_Additive
    translationSpace=BCS_BoneSpace
    alphaBoolBlend {blendOption=Linear}
    boneToModify {boneName=ponytail_r_mid_mstr}
    rotation {pitch=-15}
   }
   n43 = Blend Poses by bool(
    BlendPose_0=Component To Local(ComponentPose=@n42)
    BlendPose_1=Blend Poses by bool(
     BlendPose_0=Component To Local(ComponentPose=@n22)
     BlendPose_1=Blend Poses by bool(BlendPose_0=@n10, BlendPose_1=@n6, bActiveValue=@n11) {blendType=Linear}
     bActiveValue=Get Standing Sprint() -> (Standing Sprint:bool)
    ) {blendType=Linear}
    bActiveValue=Get Wearing Backbling() -> (Wearing Backbling:bool)
   ) {blendType=Linear}
   n44 = "Constraint - hair_bk_l_01_mstr (PP)"(
    ComponentPose=Local To Component(LocalPose=Use cached pose 'Anim Dynamics'())
    Alpha=1
    ConstraintWeights_0=0.9995
    ConstraintWeights_1=0.0005
   ) {
    alphaBoolBlend {blendOption=Linear}
    boneToModify {boneName=hair_bk_l_01_mstr}
    constraintSetup [{targetBone {boneName=spine_05}}, {targetBone {boneName=neck_01}}]
    constraintWeights [1,1]
   }
   n45 = "Constraint - ponytail_l_01_mstr (PPPPPPPP)"(
    ComponentPose=@n44
    Alpha=1
    ConstraintWeights_0=0
    ConstraintWeights_1=0.0004
    ConstraintWeights_2=0.6951
    ConstraintWeights_3=0
    ConstraintWeights_4=0.2589
    ConstraintWeights_5=0
    ConstraintWeights_6=0
    ConstraintWeights_7=0.0457
   ) {
    alphaBoolBlend {blendOption=Linear}
    boneToModify {boneName=ponytail_l_01_mstr}
    constraintSetup [
     {targetBone {boneName=spine_03}}
     {targetBone {boneName=spine_04}}
     {targetBone {boneName=spine_05}}
     {targetBone {boneName=neck_01}}
     {targetBone {boneName=clavicle_l}}
     {targetBone {boneName=upperarm_l}}
     {targetBone {boneName=upperarm_twist_01_l}}
     {targetBone {boneName=clavicle_r}}
    ]
    constraintWeights [1,1,1,1,1,1,1,1]
   }
   n46 = "Constraint - ponytail_l_02_dyn (PPPPPPPP)"(
    ComponentPose=@n45
    Alpha=1
    ConstraintWeights_0=0.0001
    ConstraintWeights_1=0.0001
    ConstraintWeights_2=0.0054
    ConstraintWeights_3=0.5939
    ConstraintWeights_4=0.3168
    ConstraintWeights_5=0.0001
    ConstraintWeights_6=0.0006
    ConstraintWeights_7=0.0829
   ) {
    alphaBoolBlend {blendOption=Linear}
    boneToModify {boneName=ponytail_l_02_dyn}
    constraintSetup [
     {targetBone {boneName=spine_02}}
     {targetBone {boneName=spine_03}}
     {targetBone {boneName=spine_04}}
     {targetBone {boneName=spine_05}}
     {targetBone {boneName=clavicle_l}}
     {targetBone {boneName=upperarm_l}}
     {targetBone {boneName=upperarm_twist_01_l}}
     {targetBone {boneName=clavicle_r}}
    ]
    constraintWeights [1,1,1,1,1,1,1,1]
   }
   n47 = "Constraint - ponytail_l_mid_mstr (PP)"(ComponentPose=@n46, Alpha=1, ConstraintWeights_0=1, ConstraintWeights_1=0) {
    alphaBoolBlend {blendOption=Linear}
    boneToModify {boneName=ponytail_l_mid_mstr}
    constraintSetup [{targetBone {boneName=spine_05}}, {targetBone {boneName=neck_01}}]
    constraintWeights [1,1]
   }
   n48 = "Constraint - ponytail_l_04_dyn (PPPPPPPP)"(
    ComponentPose=@n47
    Alpha=1
    ConstraintWeights_0=0.0087
    ConstraintWeights_1=0.0368
    ConstraintWeights_2=0.0376
    ConstraintWeights_3=0.259
    ConstraintWeights_4=0.3554
    ConstraintWeights_5=0.2211
    ConstraintWeights_6=0.011
    ConstraintWeights_7=0.0704
   ) {
    alphaBoolBlend {blendOption=Linear}
    boneToModify {boneName=ponytail_l_04_dyn}
    constraintSetup [
     {targetBone {boneName=spine_01}}
     {targetBone {boneName=spine_02}}
     {targetBone {boneName=spine_03}}
     {targetBone {boneName=spine_04}}
     {targetBone {boneName=spine_05}}
     {targetBone {boneName=clavicle_l}}
     {targetBone {boneName=upperarm_twist_01_l}}
     {targetBone {boneName=clavicle_r}}
    ]
    constraintWeights [1,1,1,1,1,1,1,1]
   }
   n49 = "Constraint - ponytail_l_05_dyn (PPPPPPPP)"(
    ComponentPose=@n48
    Alpha=1
    ConstraintWeights_0=0.0346
    ConstraintWeights_1=0.1245
    ConstraintWeights_2=0.0958
    ConstraintWeights_3=0.3398
    ConstraintWeights_4=0.1982
    ConstraintWeights_5=0.1456
    ConstraintWeights_6=0.019
    ConstraintWeights_7=0.0426
   ) {
    alphaBoolBlend {blendOption=Linear}
    boneToModify {boneName=ponytail_l_05_dyn}
    constraintSetup [
     {targetBone {boneName=spine_01}}
     {targetBone {boneName=spine_02}}
     {targetBone {boneName=spine_03}}
     {targetBone {boneName=spine_04}}
     {targetBone {boneName=spine_05}}
     {targetBone {boneName=clavicle_l}}
     {targetBone {boneName=upperarm_twist_01_l}}
     {targetBone {boneName=clavicle_r}}
    ]
    constraintWeights [1,1,1,1,1,1,1,1]
   }
   n50 = "Constraint - ponytail_l_06_dyn (PPPPPPPP)"(
    ComponentPose=@n49
    Alpha=1
    ConstraintWeights_0=0.0512
    ConstraintWeights_1=0.0731
    ConstraintWeights_2=0.2901
    ConstraintWeights_3=0.1391
    ConstraintWeights_4=0.288
    ConstraintWeights_5=0.0817
    ConstraintWeights_6=0.065
    ConstraintWeights_7=0.012
   ) {
    alphaBoolBlend {blendOption=Linear}
    boneToModify {boneName=ponytail_l_06_dyn}
    constraintSetup [
     {targetBone {boneName=pelvis}}
     {targetBone {boneName=spine_01}}
     {targetBone {boneName=spine_02}}
     {targetBone {boneName=spine_03}}
     {targetBone {boneName=spine_04}}
     {targetBone {boneName=spine_05}}
     {targetBone {boneName=clavicle_l}}
     {targetBone {boneName=clavicle_r}}
    ]
    constraintWeights [1,1,1,1,1,1,1,1]
   }
   n51 = "Constraint - ponytail_l_07_dyn (PPPPPPPP)"(
    ComponentPose=@n50
    Alpha=1
    ConstraintWeights_0=0.1623
    ConstraintWeights_1=0.0924
    ConstraintWeights_2=0.3629
    ConstraintWeights_3=0.1187
    ConstraintWeights_4=0.1958
    ConstraintWeights_5=0.036
    ConstraintWeights_6=0.0298
    ConstraintWeights_7=0.0022
   ) {
    alphaBoolBlend {blendOption=Linear}
    boneToModify {boneName=ponytail_l_07_dyn}
    constraintSetup [
     {targetBone {boneName=pelvis}}
     {targetBone {boneName=spine_01}}
     {targetBone {boneName=spine_02}}
     {targetBone {boneName=spine_03}}
     {targetBone {boneName=spine_04}}
     {targetBone {boneName=spine_05}}
     {targetBone {boneName=clavicle_l}}
     {targetBone {boneName=clavicle_r}}
    ]
    constraintWeights [1,1,1,1,1,1,1,1]
   }
   n52 = "Constraint - hair_bk_r_mstr (PP)"(ComponentPose=@n51, Alpha=1, ConstraintWeights_0=0.9995, ConstraintWeights_1=0.0005) {
    alphaBoolBlend {blendOption=Linear}
    boneToModify {boneName=hair_bk_r_mstr}
    constraintSetup [{targetBone {boneName=spine_05}}, {targetBone {boneName=neck_01}}]
    constraintWeights [1,1]
   }
   n53 = "Constraint - ponytail_r_01_mstr (PPPPP)"(
    ComponentPose=@n52
    Alpha=1
    ConstraintWeights_0=0.0011
    ConstraintWeights_1=0.5621
    ConstraintWeights_2=0.027
    ConstraintWeights_3=0.4098
    ConstraintWeights_4=0
   ) {
    alphaBoolBlend {blendOption=Linear}
    boneToModify {boneName=ponytail_r_01_mstr}
    constraintSetup [
     {targetBone {boneName=spine_04}}
     {targetBone {boneName=spine_05}}
     {targetBone {boneName=clavicle_l}}
     {targetBone {boneName=clavicle_r}}
     {targetBone {boneName=upperarm_twist_01_r}}
    ]
    constraintWeights [1,1,1,1,1]
   }
   n54 = "Constraint - ponytail_r_02_dyn (PPPPP)"(
    ComponentPose=@n53
    Alpha=1
    ConstraintWeights_0=0.01
    ConstraintWeights_1=0.5003
    ConstraintWeights_2=0.0417
    ConstraintWeights_3=0.447
    ConstraintWeights_4=0.0009
   ) {
    alphaBoolBlend {blendOption=Linear}
    boneToModify {boneName=ponytail_r_02_dyn}
    constraintSetup [
     {targetBone {boneName=spine_04}}
     {targetBone {boneName=spine_05}}
     {targetBone {boneName=clavicle_l}}
     {targetBone {boneName=clavicle_r}}
     {targetBone {boneName=upperarm_twist_01_r}}
    ]
    constraintWeights [1,1,1,1,1]
   }
   n55 = "Constraint - ponytail_r_mid_mstr (PPPPPPPP)"(
    ComponentPose=@n54
    Alpha=1
    ConstraintWeights_0=0.0106
    ConstraintWeights_1=0.0112
    ConstraintWeights_2=0.1232
    ConstraintWeights_3=0.4082
    ConstraintWeights_4=0.0752
    ConstraintWeights_5=0.3382
    ConstraintWeights_6=0.0099
    ConstraintWeights_7=0.0235
   ) {
    alphaBoolBlend {blendOption=Linear}
    boneToModify {boneName=ponytail_r_mid_mstr}
    constraintSetup [
     {targetBone {boneName=spine_02}}
     {targetBone {boneName=spine_03}}
     {targetBone {boneName=spine_04}}
     {targetBone {boneName=spine_05}}
     {targetBone {boneName=clavicle_l}}
     {targetBone {boneName=clavicle_r}}
     {targetBone {boneName=upperarm_r}}
     {targetBone {boneName=upperarm_twist_01_r}}
    ]
    constraintWeights [1,1,1,1,1,1,1,1]
   }
   n56 = "Constraint - ponytail_r_04_dyn (PPPPPPPP)"(
    ComponentPose=@n55
    Alpha=1
    ConstraintWeights_0=0.0126
    ConstraintWeights_1=0.0701
    ConstraintWeights_2=0.055
    ConstraintWeights_3=0.2816
    ConstraintWeights_4=0.2387
    ConstraintWeights_5=0.0325
    ConstraintWeights_6=0.2631
    ConstraintWeights_7=0.0463
   ) {
    alphaBoolBlend {blendOption=Linear}
    boneToModify {boneName=ponytail_r_04_dyn}
    constraintSetup [
     {targetBone {boneName=spine_01}}
     {targetBone {boneName=spine_02}}
     {targetBone {boneName=spine_03}}
     {targetBone {boneName=spine_04}}
     {targetBone {boneName=spine_05}}
     {targetBone {boneName=clavicle_l}}
     {targetBone {boneName=clavicle_r}}
     {targetBone {boneName=upperarm_twist_01_r}}
    ]
    constraintWeights [1,1,1,1,1,1,1,1]
   }
   n57 = "Constraint - ponytail_r_05_dyn (PPPPPPPP)"(
    ComponentPose=@n56
    Alpha=1
    ConstraintWeights_0=0.0328
    ConstraintWeights_1=0.1297
    ConstraintWeights_2=0.0906
    ConstraintWeights_3=0.3226
    ConstraintWeights_4=0.1942
    ConstraintWeights_5=0.0439
    ConstraintWeights_6=0.1612
    ConstraintWeights_7=0.025
   ) {
    alphaBoolBlend {blendOption=Linear}
    boneToModify {boneName=ponytail_r_05_dyn}
    constraintSetup [
     {targetBone {boneName=spine_01}}
     {targetBone {boneName=spine_02}}
     {targetBone {boneName=spine_03}}
     {targetBone {boneName=spine_04}}
     {targetBone {boneName=spine_05}}
     {targetBone {boneName=clavicle_l}}
     {targetBone {boneName=clavicle_r}}
     {targetBone {boneName=upperarm_twist_01_r}}
    ]
    constraintWeights [1,1,1,1,1,1,1,1]
   }
   n58 = "Constraint - ponytail_r_06_dyn (PPPPPPPP)"(
    ComponentPose=@n57
    Alpha=1
    ConstraintWeights_0=0.0446
    ConstraintWeights_1=0.0706
    ConstraintWeights_2=0.2895
    ConstraintWeights_3=0.1379
    ConstraintWeights_4=0.2945
    ConstraintWeights_5=0.0828
    ConstraintWeights_6=0.0233
    ConstraintWeights_7=0.0568
   ) {
    alphaBoolBlend {blendOption=Linear}
    boneToModify {boneName=ponytail_r_06_dyn}
    constraintSetup [
     {targetBone {boneName=pelvis}}
     {targetBone {boneName=spine_01}}
     {targetBone {boneName=spine_02}}
     {targetBone {boneName=spine_03}}
     {targetBone {boneName=spine_04}}
     {targetBone {boneName=spine_05}}
     {targetBone {boneName=clavicle_l}}
     {targetBone {boneName=clavicle_r}}
    ]
    constraintWeights [1,1,1,1,1,1,1,1]
   }
   n59 = "Constraint - ponytail_r_07_dyn (PPPPPPPP)"(
    ComponentPose=@n58
    Alpha=1
    ConstraintWeights_0=0.1944
    ConstraintWeights_1=0.0983
    ConstraintWeights_2=0.3985
    ConstraintWeights_3=0.1102
    ConstraintWeights_4=0.1583
    ConstraintWeights_5=0.0214
    ConstraintWeights_6=0.0039
    ConstraintWeights_7=0.0151
   ) {
    alphaBoolBlend {blendOption=Linear}
    boneToModify {boneName=ponytail_r_07_dyn}
    constraintSetup [
     {targetBone {boneName=pelvis}}
     {targetBone {boneName=spine_01}}
     {targetBone {boneName=spine_02}}
     {targetBone {boneName=spine_03}}
     {targetBone {boneName=spine_04}}
     {targetBone {boneName=spine_05}}
     {targetBone {boneName=clavicle_l}}
     {targetBone {boneName=clavicle_r}}
    ]
    constraintWeights [1,1,1,1,1,1,1,1]
   }
   n60 = "Transform (Modify) Bone - Bone: hair_bk_l_01_mstr"(ComponentPose=@n59, Alpha=1, Translation="0,-1,0", Rotation="-5,-15,0", Scale="1,1,1") {
    rotationMode=BMM_Additive
    rotationSpace=BCS_BoneSpace
    translationMode=BMM_Additive
    translationSpace=BCS_BoneSpace
    alphaBoolBlend {blendOption=Linear}
    boneToModify {boneName=hair_bk_l_01_mstr}
    rotation {pitch=-5, yaw=-15}
    translation {y=-1}
   }
   n61 = "Transform (Modify) Bone - Bone: hair_bk_r_01_dyn"(ComponentPose=@n60, Alpha=1, Translation="0,1,0", Rotation="-5,15,0", Scale="1,1,1") {
    rotationMode=BMM_Additive
    rotationSpace=BCS_BoneSpace
    translationMode=BMM_Additive
    translationSpace=BCS_BoneSpace
    alphaBoolBlend {blendOption=Linear}
    boneToModify {boneName=hair_bk_r_01_dyn}
    rotation {pitch=-5, yaw=15}
    translation {y=1}
   }
   n62 = "Transform (Modify) Bone - Bone: ponytail_l_mid_mstr"(ComponentPose=@n61, Alpha=1, Translation="0,0,0", Rotation="-15,0,0", Scale="1,1,1") {
    rotationMode=BMM_Additive
    rotationSpace=BCS_BoneSpace
    translationMode=BMM_Additive
    translationSpace=BCS_BoneSpace
    alphaBoolBlend {blendOption=Linear}
    boneToModify {boneName=ponytail_l_mid_mstr}
    rotation {pitch=-15}
   }
   n63 = "Transform (Modify) Bone - Bone: ponytail_r_mid_mstr"(ComponentPose=@n62, Alpha=1, Translation="0,0,0", Rotation="-15,0,0", Scale="1,1,1") {
    rotationMode=BMM_Additive
    rotationSpace=BCS_BoneSpace
    translationMode=BMM_Additive
    translationSpace=BCS_BoneSpace
    alphaBoolBlend {blendOption=Linear}
    boneToModify {boneName=ponytail_r_mid_mstr}
    rotation {pitch=-15}
   }
   n64 = "Transform (Modify) Bone - Bone: ponytail_l_mid_mstr"(ComponentPose=@n63, Alpha=1, Translation="0,0,0", Rotation="5,0,0", Scale="1,1,1") {
    rotationMode=BMM_Additive
    rotationSpace=BCS_BoneSpace
    translationMode=BMM_Additive
    translationSpace=BCS_BoneSpace
    alphaBoolBlend {blendOption=Linear}
    boneToModify {boneName=ponytail_l_mid_mstr}
    rotation {pitch=5}
   }
   n65 = "Transform (Modify) Bone - Bone: ponytail_l_04_dyn"(ComponentPose=@n64, Alpha=1, Translation="0,0,0", Rotation="-5,0,0", Scale="1,1,1") {
    rotationMode=BMM_Additive
    rotationSpace=BCS_BoneSpace
    translationMode=BMM_Additive
    translationSpace=BCS_BoneSpace
    alphaBoolBlend {blendOption=Linear}
    boneToModify {boneName=ponytail_l_04_dyn}
    rotation {pitch=-5}
   }
   n66 = Component To Local(ComponentPose=@n65)
   n67 = Get bRigidBodyEnabled()
   n68 = Control Rig!(
    Source=Layered blend per bone(
     BasePose=Blend Poses by bool(BlendPose_0=@n43, BlendPose_1=@n66, bActiveValue=@n67) {blendType=Linear}
     BlendPoses_0=Slot 'DefaultSlot'()
     BlendWeights_0=Get Custom Idle is Playing() -> (Custom Idle is Playing:double)
    ) {layerSetup [{branchFilters [{boneName=hat_dyn}]}]}
    Alpha
    Wearing Back Pack:bool=Get Wearing Backbling()
    Crouching:bool=Get bIsCrouching()
    Sprinting:bool=Get bIsSprinting()
    Sky Diving:bool=Get bIsSkyDiving()
    Front End:bool=Get bIsInFrontEnd()
   ) -> (Pose) {
    alphaBoolBlend {blendOption=Linear}
    controlRigAssetReference {blueprintRigClass=&/F_MED_IndieBucket_FaceAcc_ControlRig.F_MED_IndieBucket_FaceAcc_ControlRig_C}
   }
   Output Pose(Result=@n68) {layerGroup=None}
   n69 = Layered blend per bone(BasePose=Copy Pose From Mesh() {bUseAttachedParent=true}, BlendPoses_0=Copy Pose From Mesh(SourceMeshComponent=Get Skel Mesh())) {layerSetup [{branchFilters [{boneName=clavicle_l}, {boneName=upperarm_l}, {boneName=clavicle_r}, {boneName=upperarm_r}]}]}
   n70 = "Transform (Modify) Bone - Bone: dyn_r_earring_mstr"!(ComponentPose=Local To Component(LocalPose=@n69), bAlphaBoolEnabled:bool=@n1, Translation="0,0,0", Rotation="0,-20,0", Scale="1,1,1") -> (Pose) {
    alphaInputType=Bool
    rotationMode=BMM_Additive
    rotationSpace=BCS_BoneSpace
    translationMode=BMM_Additive
    translationSpace=BCS_BoneSpace
    alphaBoolBlend {blendOption=Linear}
    boneToModify {boneName=dyn_r_earring_mstr}
    rotation {yaw=-20}
   }
   n71 = Anim Dynamics - Bone: dyn_r_earring_01(ComponentPose=@n70, Alpha=1, ExternalForce="0,0,0") {
    angularDampingOverride=0.9999
    bGravityOverrideInSimSpace=true
    bOverrideAngularDamping=true
    bUseSphericalLimits=true
    simulationSpace=RootRelative
    alphaBoolBlend {blendOption=Linear}
    boundBone {boneName=dyn_r_earring_01}
    physicsBodyDefinitions [{
      sphereCollisionRadius=0
      boundBone {boneName=dyn_r_earring_01}
      boxExtents=<6 15 15>
      constraintSetup {angularLimitsMax=<5 35 15>, angularLimitsMin=<-5 -10 0>, angularTarget=<0 0 0>}
      localJointOffset=<1.5 -0.05 0.1>
     }]
    sphericalLimits [{limitRadius=4, drivingBone {boneName=head}, sphereLocalOffset=<-2 0 0>}]
   }
   n72 = Anim Dynamics - Bone: dyn_l_earring_01(ComponentPose=@n71, Alpha=1, ExternalForce="0,0,0") {
    angularDampingOverride=0.9999
    bGravityOverrideInSimSpace=true
    bOverrideAngularDamping=true
    bUseSphericalLimits=true
    simulationSpace=RootRelative
    alphaBoolBlend {blendOption=Linear}
    boundBone {boneName=dyn_l_earring_01}
    physicsBodyDefinitions [{
      sphereCollisionRadius=0
      boundBone {boneName=dyn_l_earring_01}
      boxExtents=<6 15 15>
      constraintSetup {angularLimitsMax=<5 35 20>, angularLimitsMin=<-5 -15 0>, angularTarget=<0 0 0>}
      localJointOffset=<1.5 -0.05 0.1>
     }]
    sphericalLimits [{limitRadius=4, drivingBone {boneName=head}, sphereLocalOffset=<-2 0 0>}]
   }
   n73 = Anim Dynamics - Bone: bang_center_left_a_01_dyn(ComponentPose=@n72, Alpha=1, ExternalForce="0,0,0") {
    angularDampingOverride=0.9999
    angularSpringConstant=250
    bAngularSpring=true
    bGravityOverrideInSimSpace=true
    bOverrideAngularDamping=true
    simulationSpace=RootRelative
    alphaBoolBlend {blendOption=Linear}
    boundBone {boneName=bang_center_left_a_01_dyn}
    componentAppliedLinearAccClamp=<5000 5000 5000>
    componentLinearAccScale=<1 1 1>
    componentLinearVelScale=<1 1 1>
    physicsBodyDefinitions [{
      sphereCollisionRadius=0
      boundBone {boneName=bang_center_left_a_01_dyn}
      boxExtents=<20 20 10>
      constraintSetup {angularLimitsMax=<5 2 15>, angularLimitsMin=<-5 -2 -15>, angularTarget=<0 0 0>}
      localJointOffset=<1 0 -0.5>
     }]
   }
   n74 = Anim Dynamics - Bone: bang_center_left_a_02_dyn(ComponentPose=@n73, Alpha=1, ExternalForce="0,0,0") {
    angularDampingOverride=0.9999
    angularSpringConstant=250
    bAngularSpring=true
    bGravityOverrideInSimSpace=true
    bOverrideAngularDamping=true
    simulationSpace=RootRelative
    alphaBoolBlend {blendOption=Linear}
    boundBone {boneName=bang_center_left_a_02_dyn}
    componentAppliedLinearAccClamp=<5000 5000 5000>
    componentLinearAccScale=<1 1 1>
    componentLinearVelScale=<1 1 1>
    physicsBodyDefinitions [{
      sphereCollisionRadius=0
      boundBone {boneName=bang_center_left_a_02_dyn}
      boxExtents=<20 20 10>
      constraintSetup {angularLimitsMax=<5 2 15>, angularLimitsMin=<-5 -2 -15>, angularTarget=<0 0 0>}
      localJointOffset=<1 0 -0.5>
     }]
   }
   n75 = Anim Dynamics - Bone: bang_center_left_b_01_dyn(ComponentPose=@n74, Alpha=0.5, ExternalForce="0,0,0") {
    angularDampingOverride=0.999
    angularSpringConstant=250
    bAngularSpring=true
    bGravityOverrideInSimSpace=true
    bOverrideAngularDamping=true
    simulationSpace=RootRelative
    alphaBoolBlend {blendOption=Linear}
    boundBone {boneName=bang_center_left_b_01_dyn}
    componentAppliedLinearAccClamp=<5000 5000 5000>
    componentLinearAccScale=<1 1 1>
    componentLinearVelScale=<1 1 1>
    physicsBodyDefinitions [{
      sphereCollisionRadius=0
      boundBone {boneName=bang_center_left_b_01_dyn}
      boxExtents=<20 20 10>
      constraintSetup {angularLimitsMax=<5 15 2>, angularLimitsMin=<-5 -15 -15>, angularTarget=<0 0 0>}
      localJointOffset=<1 0 -0.5>
     }]
   }
   n76 = Anim Dynamics - Bone: bang_center_left_b_02_dyn(ComponentPose=@n75, Alpha=0.5, ExternalForce="0,0,0") {
    angularDampingOverride=0.999
    angularSpringConstant=250
    bAngularSpring=true
    bGravityOverrideInSimSpace=true
    bOverrideAngularDamping=true
    simulationSpace=RootRelative
    alphaBoolBlend {blendOption=Linear}
    boundBone {boneName=bang_center_left_b_02_dyn}
    componentAppliedLinearAccClamp=<5000 5000 5000>
    componentLinearAccScale=<1 1 1>
    componentLinearVelScale=<1 1 1>
    physicsBodyDefinitions [{
      sphereCollisionRadius=0
      boundBone {boneName=bang_center_left_b_02_dyn}
      boxExtents=<20 20 10>
      constraintSetup {angularLimitsMax=<5 15 2>, angularLimitsMin=<-5 -15 -30>, angularTarget=<0 0 0>}
      localJointOffset=<1 0 -0.5>
     }]
   }
   n77 = Anim Dynamics - Bone: bang_center_left_c_01_dyn(ComponentPose=@n76, Alpha=0.5, ExternalForce="0,0,0") {
    angularDampingOverride=0.999
    angularSpringConstant=250
    bAngularSpring=true
    bGravityOverrideInSimSpace=true
    bOverrideAngularDamping=true
    simulationSpace=RootRelative
    alphaBoolBlend {blendOption=Linear}
    boundBone {boneName=bang_center_left_c_01_dyn}
    componentAppliedLinearAccClamp=<5000 5000 5000>
    componentLinearAccScale=<1 1 1>
    componentLinearVelScale=<1 1 1>
    physicsBodyDefinitions [{
      sphereCollisionRadius=0
      boundBone {boneName=bang_center_left_c_01_dyn}
      boxExtents=<20 20 10>
      constraintSetup {angularLimitsMax=<5 15 2>, angularLimitsMin=<-5 -15 -15>, angularTarget=<0 0 0>}
      localJointOffset=<1 0 -0.5>
     }]
   }
   n78 = Anim Dynamics - Bone: bang_left_a_01_dyn(ComponentPose=@n77, Alpha=1, ExternalForce="0,0,0") {
    angularDampingOverride=0.9999
    angularSpringConstant=250
    bAngularSpring=true
    bGravityOverrideInSimSpace=true
    bOverrideAngularDamping=true
    simulationSpace=RootRelative
    alphaBoolBlend {blendOption=Linear}
    boundBone {boneName=bang_left_a_01_dyn}
    componentAppliedLinearAccClamp=<5000 5000 5000>
    componentLinearAccScale=<1 1 1>
    componentLinearVelScale=<1 1 1>
    physicsBodyDefinitions [{
      sphereCollisionRadius=0
      boundBone {boneName=bang_left_a_01_dyn}
      boxExtents=<20 20 10>
      constraintSetup {angularLimitsMax=<5 15 20>, angularLimitsMin=<-5 -5 -10>, angularTarget=<0 0 0>}
      localJointOffset=<1 0 -0.5>
     }]
   }
   n79 = Anim Dynamics - Bone: bang_left_a_02_dyn(ComponentPose=@n78, Alpha=1, ExternalForce="0,0,0") {
    angularDampingOverride=0.9999
    angularSpringConstant=250
    bAngularSpring=true
    bGravityOverrideInSimSpace=true
    bOverrideAngularDamping=true
    simulationSpace=RootRelative
    alphaBoolBlend {blendOption=Linear}
    boundBone {boneName=bang_left_a_02_dyn}
    componentAppliedLinearAccClamp=<5000 5000 5000>
    componentLinearAccScale=<1 1 1>
    componentLinearVelScale=<1 1 1>
    physicsBodyDefinitions [{
      sphereCollisionRadius=0
      boundBone {boneName=bang_left_a_02_dyn}
      boxExtents=<20 20 10>
      constraintSetup {angularLimitsMax=<5 15 20>, angularLimitsMin=<-5 -5 -10>, angularTarget=<0 0 0>}
      localJointOffset=<1 0 -0.5>
     }]
   }
   n80 = Anim Dynamics - Bone: bang_left_b_01_dyn(ComponentPose=@n79, Alpha=1, ExternalForce="0,0,0") {
    angularDampingOverride=0.9999
    angularSpringConstant=250
    bAngularSpring=true
    bGravityOverrideInSimSpace=true
    bOverrideAngularDamping=true
    simulationSpace=RootRelative
    alphaBoolBlend {blendOption=Linear}
    boundBone {boneName=bang_left_b_01_dyn}
    componentAppliedLinearAccClamp=<5000 5000 5000>
    componentLinearAccScale=<1 1 1>
    componentLinearVelScale=<1 1 1>
    physicsBodyDefinitions [{
      sphereCollisionRadius=0
      boundBone {boneName=bang_left_b_01_dyn}
      boxExtents=<20 20 10>
      constraintSetup {angularLimitsMax=<5 15 20>, angularLimitsMin=<-5 -5 -10>, angularTarget=<0 0 0>}
      localJointOffset=<1 0 -0.5>
     }]
   }
   n81 = Anim Dynamics - Bone: bang_left_b_02_dyn(ComponentPose=@n80, Alpha=1, ExternalForce="0,0,0") {
    angularDampingOverride=0.9999
    angularSpringConstant=250
    bAngularSpring=true
    bGravityOverrideInSimSpace=true
    bOverrideAngularDamping=true
    simulationSpace=RootRelative
    alphaBoolBlend {blendOption=Linear}
    boundBone {boneName=bang_left_b_02_dyn}
    componentAppliedLinearAccClamp=<5000 5000 5000>
    componentLinearAccScale=<1 1 1>
    componentLinearVelScale=<1 1 1>
    physicsBodyDefinitions [{
      sphereCollisionRadius=0
      boundBone {boneName=bang_left_b_02_dyn}
      boxExtents=<20 20 10>
      constraintSetup {angularLimitsMax=<5 15 20>, angularLimitsMin=<-5 -5 -10>, angularTarget=<0 0 0>}
      localJointOffset=<1 0 -0.5>
     }]
   }
   n82 = Get front hair() -> (front hair:double)
   n83 = Anim Dynamics - Bone: bang_center_right_a_01_dyn(ComponentPose=@n81, Alpha=@n82, ExternalForce="0,0,0") {
    angularDampingOverride=0.999
    angularSpringConstant=250
    bAngularSpring=true
    bGravityOverrideInSimSpace=true
    bOverrideAngularDamping=true
    simulationSpace=RootRelative
    alphaBoolBlend {blendOption=Linear}
    boundBone {boneName=bang_center_right_a_01_dyn}
    componentAppliedLinearAccClamp=<5000 5000 5000>
    componentLinearAccScale=<1 1 1>
    componentLinearVelScale=<1 1 1>
    physicsBodyDefinitions [{
      sphereCollisionRadius=0
      boundBone {boneName=bang_center_right_a_01_dyn}
      boxExtents=<20 20 10>
      constraintSetup {angularLimitsMax=<5 15 2>, angularLimitsMin=<-5 -15 -15>, angularTarget=<0 0 0>}
      localJointOffset=<1 0 -0.5>
     }]
   }
   n84 = Anim Dynamics - Bone: bang_center_right_a_02_dyn(ComponentPose=@n83, Alpha=@n82, ExternalForce="0,0,0") {
    angularDampingOverride=0.999
    angularSpringConstant=250
    bAngularSpring=true
    bGravityOverrideInSimSpace=true
    bOverrideAngularDamping=true
    simulationSpace=RootRelative
    alphaBoolBlend {blendOption=Linear}
    boundBone {boneName=bang_center_right_a_02_dyn}
    componentAppliedLinearAccClamp=<5000 5000 5000>
    componentLinearAccScale=<1 1 1>
    componentLinearVelScale=<1 1 1>
    physicsBodyDefinitions [{
      sphereCollisionRadius=0
      boundBone {boneName=bang_center_right_a_02_dyn}
      boxExtents=<20 20 10>
      constraintSetup {angularLimitsMax=<5 15 2>, angularLimitsMin=<-5 -15 -30>, angularTarget=<0 0 0>}
      localJointOffset=<1 0 -0.5>
     }]
   }
   n85 = Anim Dynamics - Bone: bang_center_right_b_01_dyn(ComponentPose=@n84, Alpha=@n82, ExternalForce="0,0,0") {
    angularDampingOverride=0.999
    angularSpringConstant=250
    bAngularSpring=true
    bGravityOverrideInSimSpace=true
    bOverrideAngularDamping=true
    simulationSpace=RootRelative
    alphaBoolBlend {blendOption=Linear}
    boundBone {boneName=bang_center_right_b_01_dyn}
    componentAppliedLinearAccClamp=<5000 5000 5000>
    componentLinearAccScale=<1 1 1>
    componentLinearVelScale=<1 1 1>
    physicsBodyDefinitions [{
      sphereCollisionRadius=0
      boundBone {boneName=bang_center_right_b_01_dyn}
      boxExtents=<20 20 10>
      constraintSetup {angularLimitsMax=<5 15 2>, angularLimitsMin=<-5 -15 -15>, angularTarget=<0 0 0>}
      localJointOffset=<1 0 -0.5>
     }]
   }
   n86 = Anim Dynamics - Bone: bang_center_right_b_02_dyn(ComponentPose=@n85, Alpha=@n82, ExternalForce="0,0,0") {
    angularDampingOverride=0.999
    angularSpringConstant=250
    bAngularSpring=true
    bGravityOverrideInSimSpace=true
    bOverrideAngularDamping=true
    simulationSpace=RootRelative
    alphaBoolBlend {blendOption=Linear}
    boundBone {boneName=bang_center_right_b_02_dyn}
    componentAppliedLinearAccClamp=<5000 5000 5000>
    componentLinearAccScale=<1 1 1>
    componentLinearVelScale=<1 1 1>
    physicsBodyDefinitions [{
      sphereCollisionRadius=0
      boundBone {boneName=bang_center_right_b_02_dyn}
      boxExtents=<20 20 10>
      constraintSetup {angularLimitsMax=<5 15 2>, angularLimitsMin=<-5 -15 -30>, angularTarget=<0 0 0>}
      localJointOffset=<1 0 -0.5>
     }]
   }
   n87 = Anim Dynamics - Bone: bang_center_right_c_01_dyn(ComponentPose=@n86, Alpha=1, ExternalForce="0,0,0") {
    angularDampingOverride=0.999
    angularSpringConstant=250
    bAngularSpring=true
    bGravityOverrideInSimSpace=true
    bOverrideAngularDamping=true
    simulationSpace=RootRelative
    alphaBoolBlend {blendOption=Linear}
    boundBone {boneName=bang_center_right_c_01_dyn}
    componentAppliedLinearAccClamp=<5000 5000 5000>
    componentLinearAccScale=<1 1 1>
    componentLinearVelScale=<1 1 1>
    physicsBodyDefinitions [{
      sphereCollisionRadius=0
      boundBone {boneName=bang_center_right_c_01_dyn}
      boxExtents=<20 20 10>
      constraintSetup {angularLimitsMax=<5 15 2>, angularLimitsMin=<-5 -15 -30>, angularTarget=<0 0 0>}
      localJointOffset=<1 0 -0.5>
     }]
   }
   n88 = Anim Dynamics - Bone: bang_right_a_02_dyn(ComponentPose=@n87, Alpha=1, ExternalForce="0,0,0") {
    angularDampingOverride=0.9999
    angularSpringConstant=250
    bAngularSpring=true
    bGravityOverrideInSimSpace=true
    bOverrideAngularDamping=true
    simulationSpace=RootRelative
    alphaBoolBlend {blendOption=Linear}
    boundBone {boneName=bang_right_a_02_dyn}
    componentAppliedLinearAccClamp=<5000 5000 5000>
    componentLinearAccScale=<1 1 1>
    componentLinearVelScale=<1 1 1>
    physicsBodyDefinitions [{
      sphereCollisionRadius=0
      boundBone {boneName=bang_right_a_02_dyn}
      boxExtents=<20 20 10>
      constraintSetup {angularLimitsMax=<5 20 20>, angularLimitsMin=<-5 -5 -15>, angularTarget=<0 0 0>}
      localJointOffset=<1 0 -0.5>
     }]
   }
   n89 = Anim Dynamics - Bone: bang_right_b_02_dyn(ComponentPose=@n88, Alpha=1, ExternalForce="0,0,0") {
    angularDampingOverride=0.9999
    angularSpringConstant=250
    bAngularSpring=true
    bGravityOverrideInSimSpace=true
    bOverrideAngularDamping=true
    simulationSpace=RootRelative
    alphaBoolBlend {blendOption=Linear}
    boundBone {boneName=bang_right_b_02_dyn}
    componentAppliedLinearAccClamp=<5000 5000 5000>
    componentLinearAccScale=<1 1 1>
    componentLinearVelScale=<1 1 1>
    physicsBodyDefinitions [{
      sphereCollisionRadius=0
      boundBone {boneName=bang_right_b_02_dyn}
      boxExtents=<20 20 10>
      constraintSetup {angularLimitsMax=<5 15 20>, angularLimitsMin=<-5 -5 -15>, angularTarget=<0 0 0>}
      localJointOffset=<1 0 -0.5>
     }]
   }
   n90 = Anim Dynamics - Bone: bang_right_c_02_dyn(ComponentPose=@n89, Alpha=1, ExternalForce="0,0,0") {
    angularDampingOverride=0.9999
    angularSpringConstant=250
    bAngularSpring=true
    bGravityOverrideInSimSpace=true
    bOverrideAngularDamping=true
    simulationSpace=RootRelative
    alphaBoolBlend {blendOption=Linear}
    boundBone {boneName=bang_right_c_02_dyn}
    componentAppliedLinearAccClamp=<5000 5000 5000>
    componentLinearAccScale=<1 1 1>
    componentLinearVelScale=<1 1 1>
    physicsBodyDefinitions [{
      sphereCollisionRadius=0
      boundBone {boneName=bang_right_c_02_dyn}
      boxExtents=<20 20 10>
      constraintSetup {angularLimitsMax=<5 0 15>, angularLimitsMin=<-5 -15 -15>, angularTarget=<0 0 0>}
      localJointOffset=<1 0 -0.5>
     }]
   }
   n91 = Anim Dynamics - Bone: bang_right_d_02_dyn(ComponentPose=@n90, Alpha=1, ExternalForce="0,0,0") {
    angularDampingOverride=0.9999
    angularSpringConstant=250
    bAngularSpring=true
    bGravityOverrideInSimSpace=true
    bOverrideAngularDamping=true
    simulationSpace=RootRelative
    alphaBoolBlend {blendOption=Linear}
    boundBone {boneName=bang_right_d_02_dyn}
    componentAppliedLinearAccClamp=<5000 5000 5000>
    componentLinearAccScale=<1 1 1>
    componentLinearVelScale=<1 1 1>
    physicsBodyDefinitions [{
      sphereCollisionRadius=0
      boundBone {boneName=bang_right_d_02_dyn}
      boxExtents=<20 20 10>
      constraintSetup {angularLimitsMax=<5 2 15>, angularLimitsMin=<-5 -2 -15>, angularTarget=<0 0 0>}
      localJointOffset=<1 0 -0.5>
     }]
   }
   n92 = Anim Dynamics - Bone: hat_fr_01_dyn(ComponentPose=@n91, Alpha=0.5, ExternalForce="0,0,0") {
    angularDampingOverride=0.9999
    angularSpringConstant=250
    bAngularSpring=true
    bGravityOverrideInSimSpace=true
    bOverrideAngularDamping=true
    simulationSpace=RootRelative
    alphaBoolBlend {blendOption=Linear}
    boundBone {boneName=hat_fr_01_dyn}
    componentAppliedLinearAccClamp=<5000 5000 5000>
    componentLinearAccScale=<1 1 1>
    componentLinearVelScale=<1 1 1>
    physicsBodyDefinitions [{
      sphereCollisionRadius=0
      boundBone {boneName=hat_fr_01_dyn}
      boxExtents=<20 20 10>
      constraintSetup {angularLimitsMax=<1 5 15>, angularLimitsMin=<-1 -5 -5>, angularTarget=<0 0 0>}
      localJointOffset=<1 0 -0.5>
     }]
   }
   n93 = Anim Dynamics - Bone: hat_fr_r_01_dyn(ComponentPose=@n92, Alpha=0.5, ExternalForce="0,0,0") {
    angularDampingOverride=0.9999
    angularSpringConstant=250
    bAngularSpring=true
    bGravityOverrideInSimSpace=true
    bOverrideAngularDamping=true
    simulationSpace=RootRelative
    alphaBoolBlend {blendOption=Linear}
    boundBone {boneName=hat_fr_r_01_dyn}
    componentAppliedLinearAccClamp=<5000 5000 5000>
    componentLinearAccScale=<1 1 1>
    componentLinearVelScale=<1 1 1>
    physicsBodyDefinitions [{
      sphereCollisionRadius=0
      boundBone {boneName=hat_fr_r_01_dyn}
      boxExtents=<20 20 10>
      constraintSetup {angularLimitsMax=<1 10 15>, angularLimitsMin=<-1 -10 -5>, angularTarget=<0 0 0>}
      localJointOffset=<1 0 -0.5>
     }]
   }
   n94 = Anim Dynamics - Bone: hat_r_01_dyn(ComponentPose=@n93, Alpha=0.5, ExternalForce="0,0,0") {
    angularDampingOverride=0.9999
    angularSpringConstant=250
    bAngularSpring=true
    bGravityOverrideInSimSpace=true
    bOverrideAngularDamping=true
    simulationSpace=RootRelative
    alphaBoolBlend {blendOption=Linear}
    boundBone {boneName=hat_r_01_dyn}
    componentAppliedLinearAccClamp=<5000 5000 5000>
    componentLinearAccScale=<1 1 1>
    componentLinearVelScale=<1 1 1>
    physicsBodyDefinitions [{
      sphereCollisionRadius=0
      boundBone {boneName=hat_r_01_dyn}
      boxExtents=<20 20 10>
      constraintSetup {angularLimitsMax=<1 10 15>, angularLimitsMin=<-1 -10 -5>, angularTarget=<0 0 0>}
      localJointOffset=<1 0 -0.5>
     }]
   }
   n95 = Anim Dynamics - Bone: hat_bk_r_01_dyn(ComponentPose=@n94, Alpha=0.5, ExternalForce="0,0,0") {
    angularDampingOverride=0.9999
    angularSpringConstant=250
    bAngularSpring=true
    bGravityOverrideInSimSpace=true
    bOverrideAngularDamping=true
    simulationSpace=RootRelative
    alphaBoolBlend {blendOption=Linear}
    boundBone {boneName=hat_bk_r_01_dyn}
    componentAppliedLinearAccClamp=<5000 5000 5000>
    componentLinearAccScale=<1 1 1>
    componentLinearVelScale=<1 1 1>
    physicsBodyDefinitions [{
      sphereCollisionRadius=0
      boundBone {boneName=hat_bk_r_01_dyn}
      boxExtents=<20 20 10>
      constraintSetup {angularLimitsMax=<1 10 15>, angularLimitsMin=<-1 -10 -5>, angularTarget=<0 0 0>}
      localJointOffset=<1 0 -0.5>
     }]
   }
   n96 = Anim Dynamics - Bone: hat_bk_01_dyn(ComponentPose=@n95, Alpha=0.5, ExternalForce="0,0,0") {
    angularDampingOverride=0.9999
    angularSpringConstant=250
    bAngularSpring=true
    bGravityOverrideInSimSpace=true
    bOverrideAngularDamping=true
    simulationSpace=RootRelative
    alphaBoolBlend {blendOption=Linear}
    boundBone {boneName=hat_bk_01_dyn}
    componentAppliedLinearAccClamp=<5000 5000 5000>
    componentLinearAccScale=<1 1 1>
    componentLinearVelScale=<1 1 1>
    physicsBodyDefinitions [{
      sphereCollisionRadius=0
      boundBone {boneName=hat_bk_01_dyn}
      boxExtents=<20 20 10>
      constraintSetup {angularLimitsMax=<1 10 15>, angularLimitsMin=<-1 -10 -5>, angularTarget=<0 0 0>}
      localJointOffset=<1 0 -0.5>
     }]
   }
   n97 = Anim Dynamics - Bone: hat_bk_l_01_dyn(ComponentPose=@n96, Alpha=0.5, ExternalForce="0,0,0") {
    angularDampingOverride=0.9999
    angularSpringConstant=250
    bAngularSpring=true
    bGravityOverrideInSimSpace=true
    bOverrideAngularDamping=true
    simulationSpace=RootRelative
    alphaBoolBlend {blendOption=Linear}
    boundBone {boneName=hat_bk_l_01_dyn}
    componentAppliedLinearAccClamp=<5000 5000 5000>
    componentLinearAccScale=<1 1 1>
    componentLinearVelScale=<1 1 1>
    physicsBodyDefinitions [{
      sphereCollisionRadius=0
      boundBone {boneName=hat_bk_l_01_dyn}
      boxExtents=<20 20 10>
      constraintSetup {angularLimitsMax=<1 10 15>, angularLimitsMin=<-1 -10 -5>, angularTarget=<0 0 0>}
      localJointOffset=<1 0 -0.5>
     }]
   }
   n98 = Anim Dynamics - Bone: hat_l_01_dyn(ComponentPose=@n97, Alpha=0.5, ExternalForce="0,0,0") {
    angularDampingOverride=0.9999
    angularSpringConstant=250
    bAngularSpring=true
    bGravityOverrideInSimSpace=true
    bOverrideAngularDamping=true
    simulationSpace=RootRelative
    alphaBoolBlend {blendOption=Linear}
    boundBone {boneName=hat_l_01_dyn}
    componentAppliedLinearAccClamp=<5000 5000 5000>
    componentLinearAccScale=<1 1 1>
    componentLinearVelScale=<1 1 1>
    physicsBodyDefinitions [{
      sphereCollisionRadius=0
      boundBone {boneName=hat_l_01_dyn}
      boxExtents=<20 20 10>
      constraintSetup {angularLimitsMax=<1 10 15>, angularLimitsMin=<-1 -10 -5>, angularTarget=<0 0 0>}
      localJointOffset=<1 0 -0.5>
     }]
   }
   n99 = Anim Dynamics - Bone: hat_fr_l_01_dyn(ComponentPose=@n98, Alpha=0.5, ExternalForce="0,0,0") {
    angularDampingOverride=0.9999
    angularSpringConstant=250
    bAngularSpring=true
    bGravityOverrideInSimSpace=true
    bOverrideAngularDamping=true
    simulationSpace=RootRelative
    alphaBoolBlend {blendOption=Linear}
    boundBone {boneName=hat_fr_l_01_dyn}
    componentAppliedLinearAccClamp=<5000 5000 5000>
    componentLinearAccScale=<1 1 1>
    componentLinearVelScale=<1 1 1>
    physicsBodyDefinitions [{
      sphereCollisionRadius=0
      boundBone {boneName=hat_fr_l_01_dyn}
      boxExtents=<20 20 10>
      constraintSetup {angularLimitsMax=<1 10 15>, angularLimitsMin=<-1 -10 -5>, angularTarget=<0 0 0>}
      localJointOffset=<1 0 -0.5>
     }]
   }
   Save cached pose 'Anim Dynamics'(Pose=Component To Local(ComponentPose=@n99)) {cacheName=Anim Dynamics}
   Get front hair()
  ]}
 EventGraph {nodes [
   n1 = Set WPO Regional Function!(
    execute
    Material:Material Instance Dynamic Object Reference=Get Mat Instance Dynamic()
    Overlay Material:Material Instance Dynamic Object Reference=Get Overlay Mat Instance Dynamic()
    Skeletal Mesh:Skeletal Mesh Component Object Reference=Get Skel Mesh()
    Velocity Transform:Name=hat_dyn
    Velocity Interp Speed:double=Get Velocity Interp Speed()
    Collision Transform:Name=head
    Collision Transform Axis:EAxis Enum
    Invert Collision Transform Axis:bool=false
    Delta Time X:double=0
    Previous Velocity Transform Location:"Vector (by ref)"=Get R1 Velocity Transform Location()
    Previous Velocity Transform Velocity:"Vector (by ref)"=Get R1 Velocity Transform Velocity()
    Region:int=1
   ) -> (then)
   Set WPO Regional Function(
    execute=[@n9.then, @n1.then]
    Material=Get Mat Instance Dynamic()
    Overlay Material=Get Overlay Mat Instance Dynamic()
    Skeletal Mesh=Get Skel Mesh()
    Velocity Transform=hat_r_01_dyn
    Velocity Interp Speed=Get Velocity Interp Speed()
    Collision Transform=head
    Invert Collision Transform Axis=true
    Delta Time X=0
    Previous Velocity Transform Location=Get R2 Velocity Transform Location()
    Previous Velocity Transform Velocity=Get R2 Velocity Transform Velocity()
    Region=2
   )
   > n2 = Branch(Condition=Get bIsSkydiveFloating())
   > n3 = Set WPO Pho Wind Intensity(WPO Pho Wind Intensity=1)
   SetScalarParameterValue(execute=[@n4.then, @n3.then], self=Get Mat Instance Dynamic(), ParameterName=ClothSim_WindStrength, Value=Get WPO Pho Wind Intensity())
   n4 = Set WPO Pho Wind Intensity(execute=@n2.else, WPO Pho Wind Intensity=0)
   Event BlueprintUpdateAnimation()
   > n5 = Sequence!(execute) -> (then_0, then_1, then_2:Exec, then_3:Exec, then_4:Exec, then_5:Exec, then_6:Exec, then_7:Exec, then_8:Exec, then_9:Exec)
   > n6 = Gravity Override State Function DEPRECATED(CharacterPartAnimInstance=Self-Reference(), GameTimeInSeconds=GetGameTimeInSeconds(), GravityOverrideParams=Get Gravity Override Params())
   > Set Gravity Override(Gravity Override=@n6)
   n7 = CheckForBackbling()
   Set Wearing Backbling(execute=@n5.then_1, Wearing Backbling:bool=OR Boolean!(A=@n7, B=@n7.bIsBackblingACape, C:bool=@n7.bIsBackblingLong) -> (ReturnValue)) -> (Output_Get:bool)
   n8 = Branch(execute=@n5.then_2, Condition=Get Wearing Backbling())
   > Set Wearing Backpack(Wearing Backpack:Scalable Float Structure) -> (Output_Get:Scalable Float Structure)
   Set Wearing Backpack(execute=@n8.else)
   Is Valid(exec=@n5.then_3, InputObject=Get Player Pawn() -> (Player Pawn:Fort Player Pawn Athena Object Reference))
   > n9 = Set WPO Globals Function(
    Material=Get Mat Instance Dynamic()
    Overlay Material=Get Overlay Mat Instance Dynamic()
    Skeletal Mesh=Get Skel Mesh()
    "Local Noise Transform  Location"=Get Local Noise Transform Location() -> (Local Noise Transform Location:Vector)
    Local Noise Offset=Get Local Noise Offset()
   )
   n10 = Cast To PlayerPawn_Athena(execute=@n5.then_4, Object=TryGetPawnOwner())
   > Set Skel Mesh(Skel Mesh=GetSkeletalMeshForPartType(self=@n10))
   n11 = Branch(execute=@n5.then_5, Condition=Get bIsSprinting())
   > Set EG Sprinting(EG Sprinting:double=0.5) -> (Output_Get:double)
   Set EG Sprinting(execute=@n11.else, EG Sprinting=0)
   n12 = Branch(execute=@n5.then_6, Condition=Get bIsTacticalSprinting())
   > Set EG Tactical Sprinting(EG Tactical Sprinting:double=1) -> (Output_Get:double)
   Set EG Tactical Sprinting(execute=@n12.else, EG Tactical Sprinting=0)
   n13 = Get bIsCrouchSprinting()
   n14 = Get bIsSurfaceSwimming()
   n15 = Branch(execute=@n5.then_7, Condition=AND Boolean!(A=Get bIsSprinting(), B=NOT Boolean(A=@n14), C:bool=NOT Boolean(A=@n13)) -> (ReturnValue))
   > Set Standing Sprint(Standing Sprint:bool=true) -> (Output_Get:bool)
   Set Standing Sprint(execute=@n15.else, Standing Sprint=false)
   n16 = Branch(execute=@n5.then_8, Condition=AND Boolean!(A=Get bIsCrouching(), B=Get bIsMoving2D()) -> (ReturnValue))
   > Set crouch moving(crouch moving:bool=true) -> (Output_Get:bool)
   > Set head left(head left:double=GetAlphaFromReferenceTransformDelta(
     SkeletalMeshComponent=Get Skel Mesh()
     BoneName=head
     BaseName=spine_05
     DeltaComponentType=EulerX
     InputRange=(X=10,Y=35)
     TwistAxis="1,0,0"
    )) -> (Output_Get:double)
   Set crouch moving(execute=@n16.else, crouch moving=false)
   > Set head left(head left=0)
   Branch(execute=@n5.then_9, Condition=Get bIsInFrontEnd())
   > Set Custom Idle is Playing(Custom Idle is Playing:double=GetCurveValue(self=GetOwnerFrontEndAnimInstance(), CurveName=Frontend_Idle_Bypass)) -> (Output_Get:double)
   Event OnCharacterCustomizationLoaded()
   > n17 = Set Skel Mesh(Skel Mesh=GetPartSkeletalMeshComponent())
   > n18 = Get Dynamic Material Instance(SkeletalMeshComponent=@n17)
   > Set Mat Instance Dynamic(Mat Instance Dynamic=@n18)
   > n19 = GetOrCreateDynamicOverlayMaterial(MeshComponent=@n17, MaterialName=None)
   > Set Overlay Mat Instance Dynamic(Overlay Mat Instance Dynamic=@n19)
   > n20 = Branch(Condition=CharacterItemDefinitionHasAnyMetaTag(Tags=Make Array("[0]"=(TagName="Cosmetics.Variant.Property.Stage2"))))
   > Set front hair(front hair:double=0.5) -> (Output_Get:double)
   Set front hair(execute=@n20.else, front hair=1)
   Event BlueprintInitializeAnimation()
   > Set Skel Mesh(Skel Mesh=GetOwnerSkeletalMeshComponent())
   n21 = TryGetPawnOwner()
   > Is Valid(InputObject=@n21)
   > n22 = Cast To FortPlayerPawnAthena(Object=@n21)
   > Set Player Pawn(Player Pawn:Fort Player Pawn Athena Object Reference=@n22) -> (Output_Get:Fort Player Pawn Athena Object Reference)
   Branch(Condition=Get bIsInFrontEnd())
   Branch(Condition=Get bIsSkyDiving())
  ]}
]
variables [
 crouch moving {type=boolean}
 Custom Idle is Playing {type=number}
 EG Sprinting {type=number}
 EG Tactical Sprinting {type=number}
 front hair {type=number}
 Gravity Override {type=Vector}
 Gravity Override Params {
  type=GravityOverrideParamsStruct
  default {onGround_Falling {gravityOverride=<0 0 -1150>}, onGround_Moving {gravityOverride=<0 0 -1150>}, tactical_Sprint {gravityOverride=<0 0 -1150>}}
 }
 head left {type=number}
 Local Noise Offset {type=Vector}
 Local Noise Transform Location {type=Vector}
 Mat Instance Dynamic {type=MaterialInstanceDynamic}
 Overlay Mat Instance Dynamic {type=MaterialInstanceDynamic}
 Player Pawn {type=FortPlayerPawnAthena}
 R1 Velocity Transform Location {type=Vector}
 R1 Velocity Transform Velocity {type=Vector}
 R2 Velocity Transform Location {type=Vector}
 R2 Velocity Transform Velocity {type=Vector}
 Skel Mesh {type=SkeletalMeshComponent}
 Standing Sprint {type=boolean}
 Velocity Interp Speed {type=number}
 Wearing Backbling {type=boolean}
 Wearing Backpack {type=ScalableFloat, default {registryType=None, value=0, curve {curveTable=null, rowName=None}}}
 WPO Pho Wind Intensity {type=number, default=1}
]
